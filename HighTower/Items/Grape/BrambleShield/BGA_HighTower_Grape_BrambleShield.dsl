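; Bramble shield ball actor that rides on a parent pawn, with health, replicated cues and a failsafe lifespan based on heal timings.
; A looping timer sphere-traces ahead along parent velocity, applying knockback and handling surface hits that can break building pieces or bounce-launch the pawn.
; TurnOnGravity stops that timer and launches the ball as a projectile, enabling world collision after a delay.
; Ending or destruction hides it and reports absorb quest progress.
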(event Custom|SetFailsafeLifeSpan
  ;[1 1 1] Add together all the times to find how long the Ball should last then add in some extra time. Use this to set the lifespan in case some how the ball gets stuck on.
  (Actor|SetLifeSpan (Math|Float|float+float (ScalableFloat|GetValueatLevel (Variables|Default|GetRow_TimeBeforeFirstHeal)) (* (ScalableFloat|GetValueatLevel (Variables|Default|GetRow_TimeBetweenHeals)) (- (ScalableFloat|GetValueatLevel (Variables|Default|GetRow_NumberOfHeals)) 1.0)) (Variables|Default|GetAddedFailsafeTime))))

(event EventBeginPlay
  ;[1 1 1] Begin Play - Play GC
  (Game|Components|ProjectileMovement|StopSimulating (Variables|Default|GetFortProjectileMovement) (Variables|Default|GetNullHit)) ; Stop the projectile movement on spawn
  (CallFunction|SetFailsafeLifeSpan)
  (Networking|SwitchHasAuthority
    (:Authority
      (Variables|Debug|SetThicknessMult 0.25))
    (:Remote
      _))
  (Ability|ExecuteGameplayCueReplicated (Variables|GamplayCues|GetGC_Spawn))
  (Variables|Default|SetOneOverSphereRadius (/ 1.0 (Variables|Default|GetSphereRadius)))
  ;[1 0.4353 0] Start timer for Look Ahead Collisions
  (bind _returnvalue (Utilities|Time|SetTimerbyEvent (delegate Custom|LookAheadUpdateTimer) (Variables|LookAheadCollision|GetLookForwardSecondsInterval) true))
  (Variables|Default|SetTimer_RollingBallLookAheadUpdate _returnvalue))

(event Custom|LookAheadUpdateTimer
  ;[1 0.4353 0] Start timer for Look Ahead Collisions
  (CallFunction|UpdateBallRolling))

(event Custom|EndBramble
  (params (StartEndFXOnly :type Boolean))
  ;[1 1 1] End Bramble - if the shield is not destroyed, play the despawn cue and set the hide notify then silent die afetr a delay
  (CallFunction|StartDespawnFX)
  (if (not StartEndFXOnly)
    (Utilities|Time|ClearandInvalidateTimerbyHandle (Variables|Default|GetTimer_RollingBallLookAheadUpdate))
    (if (Networking|IsServer)
      (if (not (Variables|Default|GetDestroyed))
        (Utilities|FlowControl|DoOnce false
          (:Completed
            (Utilities|FlowControl|Delay 0.5)
            (bind _output_get (Variables|Default|SetEnded true))
            (Variables|Default|SetHide true)
            (Utilities|FlowControl|Sequence
              (:then_0
                (Utilities|FlowControl|Delay (Variables|Default|GetDelay_EndBeforeDeath))
                (Gameplay|SilentDie false))
              (:then_1
                (if _output_get
                  (CallFunction|SendQuestProgress))))))))))

(event Custom|SendQuestProgress
  (bind _returnvalue (Game|GetInstigator))
  ;[1 1 1] Send quest progress for damage absorb punch card
  (Networking|SwitchHasAuthority
    (:Authority
      (Utilities|IsValid _returnvalue
        (:Is%20Valid
          (bind _asfort_player_pawn (Utilities|Casting|CastToFortPlayerPawn _returnvalue))
          (if (and (Ability|Attribute|GetFloatAttribute self (Variables|Default|GetAtt_MaxHealth)) (Ability|Attribute|GetFloatAttribute self (Variables|Default|GetAtt_CurrentHealth)))
            (Quest|SendComplexCustomStatEvent (Quests|GetQuestManager (Fortnite|GetFortPlayerControllerfromActor _asfort_player_pawn) "Invalid") 0 "(GameplayTags=((TagName=\"Athena.Quests.HightowerGrapeAbsorb\")))" "(GameplayTags=)" (Math|Integer|SelectInt 1 100 (Variables|Default|GetDestroyed)))))))))

(event Custom|StartDespawnFX
  ;[1 1 1] Start Despawn FX
  (Utilities|FlowControl|DoOnce false
    (:Completed
      (if (not (Networking|IsServer))
        (Niagara|SetNiagaraVariable(Bool) (Variables|Default|GetNiagara) "Despawn" true))
      (if (not (Variables|Default|GetDestroyed))
        (Ability|ExecuteGameplayCueReplicated (Variables|GamplayCues|GetGC_Despawn))))))

(event EventOnDeathServer
  (params (Damage :type Float%20%28single-precision%29) (DamageTags :type Gameplay%20Tag%20Container%20Structure%20%28by%20ref%29) (Momentum :type Vector :default "0, 0, 0") (HitInfo :type Hit%20Result%20Structure%20%28by%20ref%29) (InstigatedBy :type Controller%20Object%20Reference) (DamageCauser :type Actor%20Object%20Reference) (EffectContext :type Gameplay%20Effect%20Context%20Handle%20Structure))
  ;[1 1 1] On Death Server - If the bramble dies check if we have already ended if not play death GC and set the notify
  (Utilities|Time|ClearandInvalidateTimerbyHandle (Variables|Default|GetTimer_RollingBallLookAheadUpdate))
  (if (not (Variables|Default|GetEnded))
    (Variables|Default|SetDestroyed true)
    (Development|PrintString "Bramble Destroyed")
    (Ability|ExecuteGameplayCueReplicated (Variables|GamplayCues|GetGC_Destroyed))
    (Variables|Default|SetHide true)
    (CallFunction|SendQuestProgress)))

(event EventOnDamageServer
  (params (Damage :type Float%20%28single-precision%29) (DamageTags :type Gameplay%20Tag%20Container%20Structure%20%28by%20ref%29) (Momentum :type Vector :default "0, 0, 0") (HitInfo :type Hit%20Result%20Structure%20%28by%20ref%29) (InstigatedBy :type Controller%20Object%20Reference) (DamageCauser :type Actor%20Object%20Reference) (EffectContext :type Gameplay%20Effect%20Context%20Handle%20Structure))
  (bind (_bblockinghit _binitialoverlap _time _distance _location _impactpoint _normal) (Collision|BreakHitResult HitInfo))
  ;[1 1 1] On Damage Server - when damage is taken play a GC at the location. After do a delay so it does not spam too much with super fast weapons.
  (Utilities|FlowControl|DoOnce false :!label _label
    (:Completed
      (Ability|ExecuteGameplayCueWithParamsReplicated (Variables|GamplayCues|GetGC_Damaged) (Ability|GameplayCue|MakeGameplayCueParameters :NormalizedMagnitude 0.0 :RawMagnitude 0.0 :Location _impactpoint :Normal _normal :GameplayEffectLevel 1 :AbilityLevel 1 :bReplicateLocationWhenUsingMinimalRepProxy false))
      (Utilities|FlowControl|Delay (Variables|GamplayCues|GetDelay_DamageGC)) ; Dont allow the Damage GC to trigger too often
      (goto _label Reset))))

(event Custom|TurnOnGravity
  (params (VelocityIn :type Vector))
  (bind _fortprojectilemovement (Variables|Default|GetFortProjectileMovement))
  (bind _sphere (Variables|Default|GetSphere))
  ;[1 1 1] turn on Gravity on the ball and turn on collision on the ball with the world.
  (Utilities|Time|ClearandInvalidateTimerbyHandle (Variables|Default|GetTimer_RollingBallLookAheadUpdate))
  (Components|Movement|SetUpdatedComponent _fortprojectilemovement (Variables|Transformation|GetRootComponent))
  (Variables|/Script/Engine.MovementComponent|SetVelocity VelocityIn _fortprojectilemovement)
  (Utilities|FlowControl|Delay (Variables|Default|GetBallPostLaunchCollisionDelay))
  (Collision|SetCollisionResponsetoChannel _sphere "ECC_Destructible" "ECR_Block")
  (Collision|SetCollisionResponsetoChannel _sphere "ECC_WorldStatic" "ECR_Block")
  (Collision|SetCollisionResponsetoChannel _sphere "ECC_WorldDynamic" "ECR_Block"))

(event Custom|UpdateBallRolling
  (bind _returnvalue (Transformation|GetActorLocation (Actor|GetAttachParentActor)))
  (bind _returnvalue_2 (Math|Vector|VectorLength (Variables|LookAheadCollision|GetVelThisFrame)))
  (bind _returnvalue_3 (Math|Vector|Normalize (Variables|LookAheadCollision|GetVelThisFrame)))
  (bind _lookforwardamount (Variables|LookAheadCollision|GetLookForwardAmount))
  (bind _returnvalue_4 (* _returnvalue_2 _lookforwardamount))
  (bind _returnvalue_5 (* _returnvalue_3 _returnvalue_4))
  (bind _returnvalue_6 (Actor|GetAttachParentActor))
  (bind _returnvalue_7 (+ _returnvalue_5 _returnvalue))
  (bind _lookaheadworldoffset (Variables|LookAheadCollision|GetLookAheadWorldOffset))
  (bind _returnvalue_8 (+ _returnvalue_7 _lookaheadworldoffset))
  (bind _velthisframe (Variables|LookAheadCollision|GetVelThisFrame))
  ;[1 0.4353 0] Looping Timer Update - Handle Look Ahead Collisions
  (Utilities|IsValid _returnvalue_6
    (:Is%20Valid
      (Utilities|FlowControl|Sequence
        (:then_0
          (Variables|LookAheadCollision|SetAttachedParent _returnvalue_6)
          (if (Variables|LookAheadCollision|GetbHasSetLastPos)
            (Variables|LookAheadCollision|SetVelThisFrame (/ (- (Transformation|GetActorLocation (Variables|LookAheadCollision|GetAttachedParent)) (Variables|LookAheadCollision|GetLastPos)) (Variables|LookAheadCollision|GetLookForwardSecondsInterval)))
            (else
              (Variables|LookAheadCollision|SetVelThisFrame (Transformation|GetVelocity _returnvalue_6))))
          (if (> _returnvalue_2 (ScalableFloat|GetValueatLevel (Variables|LookAheadCollision|GetMinVelocityForDestruction)))
            (Utilities|FlowControl|Sequence
              (:then_0
                (Utilities|FlowControl|DoOnce false
                  (:Completed
                    (Utilities|Array|Add (Variables|LookAheadCollision|GetActorsToIgnore) (Variables|LookAheadCollision|GetAttachedParent))
                    (Utilities|Array|Add (Variables|LookAheadCollision|GetActorsToIgnore) self))))
              (:then_1
                (bind (_outhits _returnvalue_9) (Collision|MultiSphereTraceForObjects (+ (Transformation|GetActorLocation) (Variables|LookAheadCollision|GetLookAheadWorldOffset)) _returnvalue_8 (- (Variables|Default|GetSphereRadius) (Variables|LookAheadCollision|GetRadiusReductionForAvoidingFloors)) (Variables|LookAheadCollision|GetTraceObjectTypes) false (Variables|LookAheadCollision|GetActorsToIgnore) "None" true "(R=0.958882,G=1.000000,B=0.946898,A=1.000000)" "(R=0.000000,G=1.000000,B=0.000000,A=1.000000)" 2e+01))
                (Utilities|FlowControl|Sequence
                  (:then_0
                    (if _returnvalue_9
                      (Variables|LookAheadCollision|SetSphereColorThisUpdate "(R=1.000000,G=0.000000,B=0.000000,A=1.000000)")
                      (Variables|Debug|SetDebugDurationThisFrame 2e+01)
                      (for _array_element _outhits
                        (Variables|LookAheadCollision|SetGE_Location (Transformation|GetActorLocation (Variables|LookAheadCollision|GetAttachedParent)))
                        (Variables|LookAheadCollision|SetGE_ImpactLocation (.ImpactPoint (Collision|BreakHitResult _array_element)))
                        (Variables|LookAheadCollision|SetGE_ImpactNormal (.ImpactNormal (Collision|BreakHitResult _array_element)))
                        ;[1 0.4353 0] Handle Hits
                        ;;[1 0.4353 0] Hit another bramble shield?  Apply Knock GEs to both
                        (bind _asbga_high_tower_grape_bramble_shield (Utilities|Casting|CastToBGA_HighTower_Grape_BrambleShield (.HitActor (Collision|BreakHitResult _array_element)))
                          (:then
                            (Utilities|IsValid (Actor|GetAttachParentActor _asbga_high_tower_grape_bramble_shield)
                              (:Is%20Valid
                                (Utilities|IsValid (Ability|GetAbilitySystemComponent (Actor|GetAttachParentActor _asbga_high_tower_grape_bramble_shield))
                                  (:Is%20Valid
                                    (Utilities|FlowControl|Sequence
                                      (:then_0
                                        (Fortnite|ApplyGameplayEffectsToActor (Variables|LookAheadCollision|GetAttachedParent) (Variables|GEs|GetGE_KnockBack) (Transformation|GetActorLocation (.HitActor (Collision|BreakHitResult _array_element))) false 0 (Ability|GetAbilitySystemComponent (Actor|GetAttachParentActor _asbga_high_tower_grape_bramble_shield))))
                                      (:then_1
                                        _)))))))
                          (:CastFailed
                            ;;[1 0.8533 0.4695] Hit a Vehicle?
                            (Utilities|Casting|CastToFortAthenaVehicle (.HitActor (Collision|BreakHitResult _array_element))
                              (:then
                                _)
                              (:CastFailed
                                ;;[1 0.8533 0.4695] Hit a Pawn?
                                (Utilities|Casting|CastToFortPawn (.HitActor (Collision|BreakHitResult _array_element))
                                  (:then
                                    (if (Utilities|NotEqual(Object) (.HitActor (Collision|BreakHitResult _array_element)) (Variables|LookAheadCollision|GetAttachedParent))
                                      _))
                                  (:CastFailed
                                    ;;[1 1 1] Hit Anything else? (including building actors) call function to handle.
                                    (bind _bounced (CallFunction|HandleSurfaceImpact (.HitActor (Collision|BreakHitResult _array_element)) (.ImpactPoint (Collision|BreakHitResult _array_element)) (.ImpactNormal (Collision|BreakHitResult _array_element)) _returnvalue_3 _velthisframe))
                                    (if _bounced
                                      (CallFunction|ExecuteImpactGCN (.ImpactPoint (Collision|BreakHitResult _array_element)) (.ImpactNormal (Collision|BreakHitResult _array_element)))
                                      (Variables|LookAheadCollision|SetImpactEnabled false)
                                      (Utilities|Time|SetTimerbyEvent (delegate Custom|SetCanBounce) (Variables|LookAheadCollision|GetMaxImpactFrequency)))))))))
                        (Networking|SwitchHasAuthority
                          (:Authority
                            ;;[1 0.8533 0.4695] Apply Knockback
                            (Fortnite|ApplyGameplayEffectsToActor (.HitActor (Collision|BreakHitResult _array_element)) (Variables|GEs|GetGE_KnockBack) (Variables|LookAheadCollision|GetGE_Location) false 0 (Ability|GetAbilitySystemComponent (Variables|LookAheadCollision|GetAttachedParent)))
                            (CallFunction|ExecuteImpactGCN (Variables|LookAheadCollision|GetGE_ImpactLocation) (Variables|LookAheadCollision|GetGE_ImpactNormal)))
                          (:Remote
                            (Utilities|Array|Add (Variables|LookAheadCollision|GetActorsToIgnore) (.HitActor (Collision|BreakHitResult _array_element))))))
                      (else
                        (Variables|LookAheadCollision|SetSphereColorThisUpdate "(R=1.000000,G=1.000000,B=1.000000,A=1.000000)")
                        (Variables|Debug|SetDebugDurationThisFrame 0.0))))
                  (:then_1
                    (if (Variables|Debug|GetDrawDebug)
                      (Rendering|Debug|DrawDebugSphere _returnvalue_8 (- (Variables|Default|GetSphereRadius) (Variables|LookAheadCollision|GetRadiusReductionForAvoidingFloors)) 12 (Variables|LookAheadCollision|GetSphereColorThisUpdate) (Variables|Debug|GetDebugDurationThisFrame) (* 3.0 (Variables|Debug|GetThicknessMult))))))))))
        (:then_1
          (Variables|LookAheadCollision|SetLastPos (Transformation|GetActorLocation (Variables|LookAheadCollision|GetAttachedParent)))
          (Variables|LookAheadCollision|SetbHasSetLastPos true))))))

(event Custom|SetCanBounce
  ;[1 0.4353 0] Handle Hits
  ;;[1 1 1] Hit Anything else? (including building actors) call function to handle.
  (Variables|LookAheadCollision|SetImpactEnabled true))

(event Custom|ExecuteImpactGCN
  (params (ImpactLoc :type Vector) (ImpactNomral :type Vector))
  ;[1 1 1] ExecuteImpactGCN
  (Utilities|FlowControl|DoOnce false :!label _label
    (:Completed
      (Ability|ExecuteGameplayCueWithParamsReplicated (Variables|GamplayCues|GetGC_Impact) (Ability|GameplayCue|MakeGameplayCueParameters :NormalizedMagnitude 0.0 :RawMagnitude 0.0 :Location ImpactLoc :Normal ImpactNomral :Instigator (Variables|LookAheadCollision|GetAttachedParent) :EffectCauser (Variables|LookAheadCollision|GetAttachedParent) :SourceObject (Variables|LookAheadCollision|GetAttachedParent) :GameplayEffectLevel 1 :AbilityLevel 1 :bReplicateLocationWhenUsingMinimalRepProxy false))
      (Utilities|FlowControl|Delay (Variables|Default|GetImpactGCFrequencyDelay))
      (goto _label Reset))))

(fn DamageOrDestroyBuildingPiece
  (params (BuildingPieceActor :type Actor%20Object%20Reference))
  (Networking|SwitchHasAuthority
    (:Authority
      _)
    (:Remote
      (Utilities|Array|Add (Variables|LookAheadCollision|GetActorsToIgnore) BuildingPieceActor)))
  (Fortnite|ApplyGameplayEffectsToActor BuildingPieceActor (Variables|GEs|GetGE_BuildingDestroy) (Variables|LookAheadCollision|GetGE_Location) false 0 (Ability|GetAbilitySystemComponent (Variables|LookAheadCollision|GetAttachedParent))))

(fn GetRotAxisfromDelta
  (params (FrameLocDelta :type Vector))
  (bind _returnvalue (Math|Vector|CrossProduct FrameLocDelta "0, 0,1.000000"))
  (bind _returnvalue_2 (Math|Vector|Normalize _returnvalue))
  (bind _output_get (Variables|SetHorizVec_X (Math|Vector|MakeVector (.x FrameLocDelta) (.y FrameLocDelta))))
  (if (Math|Vector|VectorIsNearlyZero _output_get)
    _
    (elif (Math|Vector|VectorIsNearlyZero _returnvalue_2)
      _
      (else
        (return (Math|Vector|NegateVector _returnvalue_2)))))
  (return "1.000000, 0, 0"))

(fn HandleSurfaceImpact
  (params (Optional_HitActor :type Actor%20Object%20Reference) (ImpactLocation :type Vector) (ImpactNormal :type Vector) (CurrentVelocityNormalized :type Vector) (CurrentVelocity :type Vector))
  (bind _localimpactnormal (LocalVariables|GetLocalImpactNormal))
  (bind _returnvalue (Math|Vector|DotProduct _localimpactnormal CurrentVelocityNormalized))
  (bind (_localcurrentvelocity_x _localcurrentvelocity_y _localcurrentvelocity_z) (Math|Vector|BreakVector (Variables|GetLocalCurrentVelocity_X)))
  ;[1 0.8533 0.4695] Handle landscape and building piece collisions?  Destroy building pieces that are hit directly.
  (LocalVariables|SetLocalImpactLoc ImpactLocation)
  (LocalVariables|SetLocalImpactNormal ImpactNormal)
  (LocalVariables|SetLocalCurrentVelocity CurrentVelocity)
  (LocalVariables|SetlocalCausedImpact false)
  (if (Variables|Debug|GetDrawDebug)
    (Utilities|FlowControl|Sequence
      (:then_0
        _)
      (:then_1
        (Development|PrintString (Utilities|String|Append "Impact!  Actor: " (Utilities|GetDisplayName Optional_HitActor) (Utilities|String|ToString(Float) _returnvalue)) true true "(R=1.000000,G=0.794383,B=0.000000,A=1.000000)" 2e+01)
        (if (LocalVariables|GetIsFloorHit)
          (if (LocalVariables|GetHitIsDirect)
            (LocalVariables|SetLocalColor "(R=1.000000,G=0.620833,B=0.000000,A=1.000000)")
            (else
              (LocalVariables|SetLocalColor "(R=0.000000,G=0.941007,B=1.000000,A=1.000000)")))
          (elif (LocalVariables|GetHitIsDirect)
            (LocalVariables|SetLocalColor "(R=1.000000,G=0.000000,B=0.071486,A=1.000000)")
            (else
              (LocalVariables|SetLocalColor "(R=1.000000,G=1.000000,B=1.000000,A=1.000000)"))))
        (Rendering|Debug|DrawDebugLine (LocalVariables|GetLocalImpactLoc) (+ (LocalVariables|GetLocalImpactLoc) "0,0.000000,5.000000") (LocalVariables|GetLocalColor) 5.0 (* 5.0 (Variables|Debug|GetThicknessMult)))
        (Rendering|Debug|DrawDebugLine (LocalVariables|GetLocalImpactLoc) (+ (LocalVariables|GetLocalImpactLoc) (* (LocalVariables|GetLocalImpactNormal) 1e+02)) (LocalVariables|GetLocalColor) 5.0 (* 2.0 (Variables|Debug|GetThicknessMult))))
      (:then_2
        (return (LocalVariables|GetlocalCausedImpact))))
    (else
      (Utilities|FlowControl|Sequence
        (:then_0
          _)
        (:then_1
          (return (LocalVariables|GetlocalCausedImpact))))))
  (bind _output_get (LocalVariables|SetIsFloorHit (> (Math|Vector|DotProduct (LocalVariables|GetLocalImpactNormal) "0.000000,0.000000,1.000000") 0.5)))
  (if _output_get
    ;;[1 1 1] Floor Hit
    (LocalVariables|SetHitIsDirect (< _returnvalue (* (ScalableFloat|GetValueatLevel (Variables|LookAheadCollision|GetDirectHitThresholdFloor)) -1.0)))
    (else
      ;;[1 1 1] Wall Hit
      (LocalVariables|SetHitIsDirect (< _returnvalue (* (ScalableFloat|GetValueatLevel (Variables|LookAheadCollision|GetDirectHitThresholdWalls)) -1.0)))
      ;;[1 0.3385 0.3385] Special cases stairs, rifts, traps, etc...
      (Utilities|Casting|CastToFortAthenaRiftPortal Optional_HitActor
        (:then
          _)
        (:CastFailed
          (Utilities|Casting|CastToBGA_Athena_Sentry_Parent Optional_HitActor
            (:then
              _)
            (:CastFailed
              (Utilities|Casting|CastToBuildingTrap Optional_HitActor
                (:then
                  _)
                (:CastFailed
                  (bind _asbuilding_stairs (Utilities|Casting|CastToBuildingStairs Optional_HitActor)
                    (:then
                      (if (Variables|LookAheadCollision|GetbHasSetLastPos)
                        ;;;[0.1927 0.06418 0.06418] Make sure we're coming from the back of the stairs
                        (if (< (Math|Vector|DotProduct (- (Variables|LookAheadCollision|GetLastPos) (+ (Transformation|GetActorLocation _asbuilding_stairs) (+ (* (.Y (Math|Rotator|GetAxes (Transformation|GetActorRotation _asbuilding_stairs))) 256.0) (* (.Z (Math|Rotator|GetAxes (Transformation|GetActorRotation _asbuilding_stairs))) 192.0)))) (+ (* (.Y (Math|Rotator|GetAxes (Transformation|GetActorRotation _asbuilding_stairs))) 256.0) (* (.Z (Math|Rotator|GetAxes (Transformation|GetActorRotation _asbuilding_stairs))) 192.0))) 0.0)
                          _)))
                    (:CastFailed
                      _))
                  (Utilities|Casting|CastToBuildingActor Optional_HitActor :!label _label_4
                    (:then
                      _)
                    (:CastFailed
                      (Utilities|Casting|CastToBuildingProp Optional_HitActor
                        (:then
                          _)
                        (:CastFailed
                          (Utilities|Casting|CastToAthena_Prop_ParentBuildingContainerBlueprint Optional_HitActor
                            (:then
                              (if (Variables|Actor|GetbCanBeDamaged Optional_HitActor) :!label _label_3
                                (if (LocalVariables|GetIsFloorHit)
                                  (if (LocalVariables|GetHitIsDirect)
                                    (if (Variables|LookAheadCollision|GetBreakOnDirectFloorHit)
                                      (CallFunction|DamageorDestroyBuildingPiece Optional_HitActor :!label _label_2))
                                    (Networking|SwitchHasAuthority :!label _label
                                      (:Authority
                                        _)
                                      (:Remote
                                        _)) ; Disconnect
                                    ;;[0.9115 0.8436 0.2848] Test if impulse is enabled...
                                    (if (Variables|LookAheadCollision|GetImpactEnabled)
                                      (if (Variables|LookAheadCollision|GetCanImpactAffectVelocity)
                                        (if (not (LocalVariables|GetIsFloorHit))
                                          _
                                          (elif (< _localcurrentvelocity_z (* (ScalableFloat|GetValueatLevel (Variables|LookAheadCollision|GetMinVertSpeedForBounce)) -1.0))
                                            _))
                                        (LocalVariables|SetVelocityAfterBounce (Math|Vector|MirrorVectorbyNormal (LocalVariables|GetLocalCurrentVelocity) (LocalVariables|GetLocalImpactNormal)))
                                        (if (LocalVariables|GetHitIsDirect)
                                          (if (LocalVariables|GetIsFloorHit)
                                            (bind _output_get_2 (LocalVariables|SetBounceMult (ScalableFloat|GetValueatLevel (Variables|LookAheadCollision|GetBreakingFloorBounceMultiplier))))
                                            (LocalVariables|SetVelocityAfterBounce (* (LocalVariables|GetVelocityAfterBounce) _output_get_2))
                                            (else
                                              (bind _output_get_3 (LocalVariables|SetBounceMult (ScalableFloat|GetValueatLevel (Variables|LookAheadCollision|GetBreakingWallBounceMultiplier))))
                                              (LocalVariables|SetVelocityAfterBounce (* (LocalVariables|GetVelocityAfterBounce) _output_get_3))))
                                          (elif (LocalVariables|GetIsFloorHit)
                                            (bind _output_get_4 (LocalVariables|SetBounceMult (ScalableFloat|GetValueatLevel (Variables|LookAheadCollision|GetNonBreakingFloorBounceMultiplier))))
                                            (LocalVariables|SetVelocityAfterBounce (* (LocalVariables|GetVelocityAfterBounce) _output_get_4))
                                            (else
                                              (bind _output_get_5 (LocalVariables|SetBounceMult (ScalableFloat|GetValueatLevel (Variables|LookAheadCollision|GetNonBreakingWallBounceMultiplier))))
                                              (LocalVariables|SetVelocityAfterBounce (* (LocalVariables|GetVelocityAfterBounce) _output_get_5)))))
                                        (bind _asfort_pawn (Utilities|Casting|CastToFortPawn (Variables|LookAheadCollision|GetAttachedParent)))
                                        (Networking|ForceNetUpdate _asfort_pawn)
                                        (Pawn|Character|LaunchCharacterJump _asfort_pawn (LocalVariables|GetVelocityAfterBounce) true true false true)
                                        (if (Variables|Debug|GetDrawDebugLaunch)
                                          (Rendering|Debug|DrawDebugLine (LocalVariables|GetLocalImpactLoc) (+ (LocalVariables|GetLocalImpactLoc) "0,0.000000,5.000000") (LocalVariables|GetLocalColor) 5.0 (* 5.0 (Variables|Debug|GetThicknessMult)))
                                          (Rendering|Debug|DrawDebugLine (LocalVariables|GetLocalImpactLoc) (+ (LocalVariables|GetLocalImpactLoc) (LocalVariables|GetVelocityAfterBounce)) (LocalVariables|GetLocalColor) 5.0 (* 2.0 (Variables|Debug|GetThicknessMult))))
                                        (LocalVariables|SetlocalCausedImpact true)
                                        (else
                                          (LocalVariables|SetlocalCausedImpact (LocalVariables|GetHitIsDirect)))))
                                    (elif (Variables|LookAheadCollision|GetBreakOnIndirectFloorHit)
                                      _
                                      (else
                                        (goto _label))))
                                  (elif (LocalVariables|GetHitIsDirect)
                                    (if (not (Variables|DefaultLookAheadCollision|GetBreakOnDirectWallHit))
                                      (goto _label))
                                    (elif (Variables|DefaultLookAheadCollision|GetBreakOnIndirectWallHit)
                                      _
                                      (else
                                        (goto _label)))))
                                (goto _label_2)))
                            (:CastFailed
                              _))
                          (goto _label)))))
                  (goto _label_3)))))))
      (return false)))
  (goto _label_4))

(fn OnRep_Hide
  (bind _hide (Variables|Default|GetHide))
  (Rendering|SetActorHiddenInGame _hide)
  (Collision|SetActorEnableCollision (not _hide)))

(fn ConstructionScript)
