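# M_Hightower_Floaties is an unlit translucent Niagara sprite material for floating particles in the Bramble Shield effect.
# A Sprite dynamic parameter shifts half-scale UVs horizontally between cells of the Floaties texture.
# A camera-distance fade blends the green channel of two vertically offset cells.
# That mask times particle alpha is faded out near the camera, then depth-faded against intersecting geometry.
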Material /HighTower/Items/Grape/BrambleShield/FX/Material/M_Hightower_Floaties
graph {
 properties {bUsedWithNiagaraSprites=true, blendMode=BLEND_Translucent, pixelDepthOffsetMode=PDOM_Legacy, shadingModel=MSM_Unlit, usedShadingModels=Unlit}
 nodes [
  n1 = ParticleColor()
  n2 = Multiply(A=TextureCoordinate()) {constB=0.5}
  n3 = LinearInterpolate(
   A=Add(A=@n2, B=Constant2Vector() {r=-1}) {constB=-0.25}
   B=Add(A=@n2, B=Constant2Vector() {r=-0.5})
   Alpha=DynamicParameter() {defaultValue=<rgba 1 0 1 1>, paramNames [Sprite,Param2,Param3,Param4]}.Sprite
  ) {constAlpha=0}
  n4 = Constant2Vector() {g=0.5}
  n5 = ScalarParameter() {defaultValue=800, parameterName=DistanceToFade}
  n6 = LinearInterpolate(
   A=TextureSample(UVs=Add(A=@n3, B=@n4)) {texture=/HighTower/Items/Grape/BrambleShield/FX/Texture/T_Hightower_Grape_Floaties}.G
   B=TextureSample(UVs=@n3) {texture=&T_Hightower_Grape_Floaties}.G
   Alpha=Clamp(None=MaterialFunctionCall(Distance=@n5) {materialFunction=&Near_Camera_Fade}.Result)
  )
  n7 = ScalarParameter() {defaultValue=250, parameterName=FadeWhenCloseDistance}
  DepthFade(
   Opacity=Multiply(
    A=Multiply(A=@n1.A, B=@n6)
    B=MaterialFunctionCall(Distance=@n7) {materialFunction=&Near_Camera_Fade}.Result
   )
   FadeDistance=ScalarParameter() {defaultValue=45, parameterName=Depth Fade}
  )
 ]
}
