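# Masked two-sided material for the HighTower Grape Bramble Shield, used on Niagara mesh particles.
# Static switch IsBramble picks bramble or inner-shield textures, each using one mask channel for metallic, specular and roughness.
# Both multiply alpha by a particle-driven SphereMask for reveal or dissolve.
# The bramble masks actor-relative world position around Offset via MaskLerpPrimary; the inner shield masks turbulence-distorted UVs around their center via MaskLerpSecondary.
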Material /HighTower/Items/Grape/BrambleShield/FX/Material/M_Hightower_Grape
graph {
 properties {
  bUseMaterialAttributes=true
  bUsedWithNiagaraMeshParticles=true
  blendMode=BLEND_Masked
  pixelDepthOffsetMode=PDOM_Legacy
  twoSided=true
  usedShadingModels=Default Lit
 }
 nodes [
  Constant() {r=1}
  n1 = TextureSample() {texture=/HighTower/Items/Grape/BrambleShield/Assets/Textures/T_HighTower_Bramble_Shield_D}
  n2 = TextureSample() {samplerType=SAMPLERTYPE_Masks, texture=/HighTower/Items/Grape/BrambleShield/Assets/Textures/T_HighTower_Bramble_Shield_S}
  n3 = Subtract(A=WorldPosition().XYZ, B=ActorPositionWS())
  n4 = VectorParameter() {parameterName=Offset, defaultValue=<rgba 50 0 0 1>}
  n5 = DynamicParameter() {defaultValue=<rgba 40 1 1 1>, paramNames [MaskLerpPrimary, MaskLerpSecondary, Param3, Param4]}
  n6 = ScalarParameter() {defaultValue=1, parameterName=MaskHardness}
  n7 = SetMaterialAttributes!(
   MaterialAttributes
   Base Color=@n1.RGB
   Metallic=@n2.G
   Specular=@n2.G
   Roughness=@n2.G
   Normal=TextureSample() {samplerType=SAMPLERTYPE_Normal, texture=/HighTower/Items/Grape/BrambleShield/Assets/Textures/T_HighTower_Bramble_Shield_N}.RGB
   Opacity Mask=Multiply(A=@n1.A, B=SphereMask(A=@n3, B=@n4.RGB, Radius=@n5.MaskLerpPrimary, Hardness=@n6))
  ) {attributeSetTypes [
    #1
    #2
    #3
    #4
    #5
    #6
   ]}
  n8 = TextureSample() {texture=/HighTower/Items/Grape/BrambleShield/Assets/Textures/T_Hightower_Grape_Shield_Inner_D}
  n9 = TextureSample() {samplerType=SAMPLERTYPE_Masks, texture=/HighTower/Items/Grape/BrambleShield/Assets/Textures/T_Hightower_Grape_Shield_Inner_S}
  n10 = TextureSample() {samplerType=SAMPLERTYPE_Normal, texture=/Game/Effects/Fort_Effects/Textures/Plasma/T_Turbulence_Seamless_N}
  n11 = Add(A=TextureCoordinate(), B=Multiply(A=ComponentMask(None=@n10.RGB) {g=true, r=true}) {constB=0.175})
  n12 = Constant() {r=0.5}
  n13 = DynamicParameter() {defaultValue=<rgba 40 1 1 1>, paramNames [MaskLerpPrimary, MaskLerpSecondary, Param3, Param4]}
  n14 = ScalarParameter() {defaultValue=1, parameterName=MaskHardness}
  n15 = SetMaterialAttributes(
   Base Color=@n8.RGB
   Metallic=@n9.G
   Specular=@n9.G
   Roughness=@n9.G
   Normal=TextureSample() {samplerType=SAMPLERTYPE_Normal, texture=/HighTower/Items/Grape/BrambleShield/Assets/Textures/T_Hightower_Grape_Shield_Inner_N}.RGB
   Opacity Mask=Multiply(A=@n8.A, B=SphereMask(A=@n11, B=@n12, Radius=@n13.MaskLerpSecondary, Hardness=@n14))
  ) {attributeSetTypes [
    #1
    #2
    #3
    #4
    #5
    #6
   ]}
  StaticSwitchParameter(True=@n7, False=@n15) {defaultValue=true, parameterName=IsBramble}
 ]
}
