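# Niagara effect for the Hightower Grape item's BrambleShield ball, with all five emitters disabled in this file.
# Rolly_Ball is one mesh particle tracking the system location, with a scratch correction module and age-driven material parameters, though its integrator references are null and disabled.
# Floaty_Bits scales sprite spawning with system velocity and adds pulsing flare variables.
# Delayed one-shot bursts supply outward spikes, floaty bits and an expanding, fading wind mesh.
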NiagaraSystem /HighTower/Items/Grape/BrambleShield/FX/NS_Hightower_Grape_Ball
dependencies {
 usedDataInterfaces [&NiagaraDataInterfaceCurve, &NiagaraDataInterfaceVector2DCurve]
 usedDynamicInputs []
 usedModules [
  &self:SystemSpawnScript.SystemScriptSource.SystemScriptGraph.NiagaraNodeAssignment_0.SetVariables_#1
  &SystemState
 ]
}
emitters {
 DirectionalBurstSpikes {
  bEnabled=false
  emitterUpdateScript [
   &EmitterState(
    Enable Distance Culling=true
    Life Cycle Mode=::NewEnumerator1
    Loop Behavior=::NewEnumerator1
    Loop Duration=1
    Max Distance Response=::KillAfterParticlesFinish
    Module.bUseMaxDistance=true
    Module.bUseMinDistance=false
    Scalability Mode=::NewEnumerator1
    UseLoopDelay=false
   )
   &SpawnBurst_Instantaneous(Spawn Count=&UniformRangedInt(Minimum=18, Maximum=24), Spawn Time=0.125, Use Spawn Probability=false)
  ]
  particleSpawnScript [
   /Niagara/Modules/Spawn/Initialization/InitializeParticle(
    Color=&UniformRangedLinearColor(Minimum=<rgba 1 1 1 0.75>, Maximum=<rgba 1 1 1 0.452>)
    Lifetime=&UniformRangedFloat(Minimum=0.15, Maximum=0.25)
    Mass=&UniformRangedFloat(Minimum=0.75, Maximum=1.75)
    Module.Write Color=true
    Module.Write Lifetime=true
    Module.Write Mass=true
    Module.Write Material Random=true
    Module.Write Position=true
    Module.Write Scale=false
    Module.Write SpriteRotation=false
    Module.Write SpriteSize=true
    Position=&SimulationPosition()
    Sprite Size=&UniformRangedVector2D(Minimum=<10 80>, Maximum=<50 150>)
   )
   /Niagara/Modules/Spawn/Location/SphereLocation(MaskSpawnGroup=false, Module.Override Randomness=false, Surface Only=false)
   &AddVelocityFromPoint(Module.Use Falloff Distance=false, Velocity Strength=500)
   &DynamicMaterialParameters(
    Index 0 Param 1=50
    Module.Param0WriteEnabled=true
    Module.Param1WriteEnabled=true
    Module.Param2WriteEnabled=true
    Module.Param3WriteEnabled=true
   )
  ]
  particleUpdateScript [
   &ParticleState()
   &GravityForce() {enabled=false}
   &Drag(Use Linear Drag=true, Use Rotational Drag=false) {enabled=false}
   &ScaleColor(
    Scale Alpha=&FloatFromCurve(FloatCurve=DataInterface() {propertyValues {Curve {keys [
         {interpMode=RCIM_Cubic, leaveTangent=-0.0594, tangentMode=RCTM_Break, time=-0.0029, value=1.0303}
         {arriveTangent=-0.936, interpMode=RCIM_Cubic, leaveTangent=-0.936, tangentMode=RCTM_User, time=0.0887, value=0.3527}
         <at 1.0011 0 RCIM_Cubic>
        ]}}})
    ScaleA=true
    ScaleRGB=true
   )
   &SolveForcesAndVelocity(Clamp Velocity=false, Limit Acceleration=false)
   &ScaleSpriteSize(Scale Factor=&Vector2DFromCurve(
     Vector2Curve=DataInterface() {propertyValues {
       XCurve {keys [<at 0.0014 0.4615>,<at 1 1>]}
       YCurve {keys [{arriveTangent=1.6918, interpMode=RCIM_Cubic, leaveTangent=1.6918, tangentMode=RCTM_User, time=-0.0006, value=0.4523}, <at 1 1>]}
      }}
     ScaleCurve=<1 1>
    ))
  ]
  rendererClasses [&NiagaraSpriteRendererProperties]
  renderers [{rendererClass=&NiagaraSpriteRendererProperties}]
 }
 Floaty_Bits {
  bEnabled=false
  emitterUpdateScript [
   &EmitterState(
    Enable Distance Culling=true
    Inactive Response=::NewEnumerator2
    Life Cycle Mode=::NewEnumerator1
    Loop Duration=1
    MaxDistance=2000
    Module.bUseMaxDistance=true
    Module.bUseMinDistance=false
    Reset Age On Awaken=true
    Scalability Mode=::NewEnumerator1
    UseLoopDelay=false
   )
   &SpawnRate(
    SpawnRate=&Lerp_Float(A=0, B=40, Alpha=&VectorLength(Vector=$System.SystemVelocity))
    Use Spawn Probability=false
   )
  ]
  particleSpawnScript [
   null() {enabled=false, module=FN_Orbit_Init_Variables_v2}
   &InitializeParticle(
    Color=<rgba 1 0.8419 0.0781 0.5>
    Lifetime=&UniformRangedFloat(Minimum=2, Maximum=4.85)
    Module.Write Color=true
    Module.Write Lifetime=true
    Module.Write Mass=true
    Module.Write Material Random=true
    Module.Write Position=false
    Module.Write Scale=false
    Module.Write SpriteRotation=true
    Module.Write SpriteSize=true
    Position=&SimulationPosition()
    Sprite Rotation=&UniformRangedFloat(Minimum=-20, Maximum=20)
    Sprite Size=&Vector2DFromFloat(Value=&UniformRangedFloat(Minimum=6, Maximum=15))
   )
   &SphereLocation(MaskSpawnGroup=false, Module.Override Randomness=false, Sphere Radius=85, Surface Only=true, Surface Only Band Thickness=8)
   &AddVelocityFromPoint(Module.Use Falloff Distance=false, Velocity Strength=55)
   &DynamicMaterialParameters(
    Index 0 Param 1=&UniformAOrBFloat(A=0)
    Module.Param0WriteEnabled=true
    Module.Param1WriteEnabled=false
    Module.Param2WriteEnabled=false
    Module.Param3WriteEnabled=false
   )
  ]
  particleUpdateScript [
   &ParticleState()
   /SetVariables(Particles.FlareColor=&MultiplyLinearColor(
     Linear Color=$Particles.Color
     Scale Factor=&MakeVector4(
      X=1
      Y=1
      Z=1
      W=&FloatFromCurve(FloatCurve=DataInterface() {propertyValues {Curve {keys [<at 0 0>,<at 0.0676 1.0023>,<at 0.4738 0.9198>,<at 1 0>]}}}, Scale Curve=0.025)
     )
    )) {isSetParameters=true, module=SetVariables_#2}
   /SetVariables(Particles.FlareSize:Vector2f=&Multiply_Vector2(
     A=$Particles.Initial.SpriteSize
     B=&Multiply_Vector2(
      A=&Vector2DFromFloat(Value=&Lerp_Float(A=2, B=2.25, Alpha=&Sine(Normalized Angle=$Emitter.Age, Period=1, Scale=1, Bias=1)))
      B=&Vector2DFromFloat(Value=&FloatFromCurve(FloatCurve=DataInterface() {propertyValues {Curve {keys [<at 0 0>,<at 0.0676 1.0023>,<at 0.4738 0.9198>,<at 1 0>]}}}))
     )
    )) {isSetParameters=true, module=SetVariables_#3}
   &Drag(Drag=0.875, Use Linear Drag=true, Use Rotational Drag=false)
   &GravityForce(Gravity=&UniformRangedVector(Minimum=<50 50 50>, Maximum=<-50 -50 -50>))
   &SolveForcesAndVelocity(Clamp Velocity=false, Limit Acceleration=false)
   &ScaleSpriteSize(Scale Factor=&Vector2DFromFloat(Value=&FloatFromCurve(FloatCurve=DataInterface() {propertyValues {Curve {keys [
          {arriveTangent=31.8914, interpMode=RCIM_Cubic, leaveTangent=31.8914, tangentMode=RCTM_User, time=0, value=0}
          <at 0.0913 1 RCIM_Cubic>
          <at 1 0 RCIM_Cubic>
         ]}}})))
   &SpriteRotationRate(Rotation Rate=&Lerp_Float(A=-30, B=30, Alpha=$Particles.Initial.MaterialRandom))
   null() {enabled=false, module=FN_Orbit_v3}
  ]
  rendererClasses [&NiagaraSpriteRendererProperties]
  renderers [
   {rendererClass=&NiagaraSpriteRendererProperties}
   {rendererClass=&NiagaraSpriteRendererProperties, rendererIndex=1}
  ]
 }
 Floaty_Bits_Burst {
  bEnabled=false
  emitterUpdateScript [
   &EmitterState(
    Enable Distance Culling=true
    Life Cycle Mode=::NewEnumerator1
    Loop Behavior=::NewEnumerator1
    Loop Duration=1
    Max Distance Response=::KillAfterParticlesFinish
    MaxDistance=3500
    Module.bUseMaxDistance=true
    Module.bUseMinDistance=false
    Reset Age On Awaken=true
    Scalability Mode=::NewEnumerator1
    UseLoopDelay=false
   )
   &SpawnBurst_Instantaneous(Spawn Count=30, Spawn Time=0.125, Use Spawn Probability=false)
  ]
  particleSpawnScript [
   null() {enabled=false, module=FN_Orbit_Init_Variables_v2}
   &InitializeParticle(
    Color=<rgba 1 0.8419 0.0781 0.5>
    Lifetime=&UniformRangedFloat(Minimum=1.25, Maximum=3)
    Module.Write Color=true
    Module.Write Lifetime=true
    Module.Write Mass=true
    Module.Write Material Random=true
    Module.Write Position=false
    Module.Write Scale=false
    Module.Write SpriteRotation=true
    Module.Write SpriteSize=true
    Position=&SimulationPosition()
    Sprite Rotation=&UniformRangedFloat(Minimum=-20, Maximum=20)
    Sprite Size=&Vector2DFromFloat(Value=&UniformRangedFloat(Minimum=6, Maximum=15))
   )
   &SphereLocation(MaskSpawnGroup=false, Module.Override Randomness=false, Sphere Radius=85, Surface Only=true, Surface Only Band Thickness=8)
   &AddVelocityFromPoint(Module.Use Falloff Distance=false, Velocity Strength=500)
   &DynamicMaterialParameters(
    Index 0 Param 1=&UniformAOrBFloat(A=0)
    Module.Param0WriteEnabled=true
    Module.Param1WriteEnabled=false
    Module.Param2WriteEnabled=false
    Module.Param3WriteEnabled=false
   )
  ]
  particleUpdateScript [
   &ParticleState()
   /SetVariables(Particles.FlareColor=&MultiplyLinearColor(
     Linear Color=$Particles.Color
     Scale Factor=&MakeVector4(
      X=1
      Y=1
      Z=1
      W=&FloatFromCurve(FloatCurve=DataInterface() {propertyValues {Curve {keys [<at 0 0>,<at 0.0676 1.0023>,<at 0.4738 0.9198>,<at 1 0>]}}}, Scale Curve=0.025)
     )
    )) {isSetParameters=true, module=SetVariables_#2}
   /SetVariables(Particles.FlareSize:Vector2f=&Multiply_Vector2(
     A=$Particles.Initial.SpriteSize
     B=&Multiply_Vector2(
      A=&Vector2DFromFloat(Value=&Lerp_Float(A=2, B=2.25, Alpha=&Sine(Normalized Angle=$Emitter.Age, Period=1, Scale=1, Bias=1)))
      B=&Vector2DFromFloat(Value=&FloatFromCurve(FloatCurve=DataInterface() {propertyValues {Curve {keys [<at 0 0>,<at 0.0676 1.0023>,<at 0.4738 0.9198>,<at 1 0>]}}}))
     )
    )) {isSetParameters=true, module=SetVariables_#3}
   &Drag(Drag=5, Use Linear Drag=true, Use Rotational Drag=false)
   &GravityForce(Gravity=&UniformRangedVector(Minimum=<50 50 50>, Maximum=<-50 -50 -50>))
   &SolveForcesAndVelocity(Clamp Velocity=false, Limit Acceleration=false)
   &ScaleSpriteSize(Scale Factor=&Vector2DFromFloat(Value=&FloatFromCurve(FloatCurve=DataInterface() {propertyValues {Curve {keys [
          {arriveTangent=31.8914, interpMode=RCIM_Cubic, leaveTangent=31.8914, tangentMode=RCTM_User, time=0, value=0}
          <at 0.0913 1 RCIM_Cubic>
          <at 1 0 RCIM_Cubic>
         ]}}})))
   &SpriteRotationRate(Rotation Rate=&Lerp_Float(A=-30, B=30, Alpha=$Particles.Initial.MaterialRandom))
   null() {enabled=false, module=FN_Orbit_v3}
  ]
  rendererClasses [&NiagaraSpriteRendererProperties]
  renderers [
   {rendererClass=&NiagaraSpriteRendererProperties}
   {rendererClass=&NiagaraSpriteRendererProperties, rendererIndex=1}
  ]
 }
 Rolly_Ball {
  bEnabled=false
  emitterUpdateScript [
   &EmitterState(
    Inactive Response=::NewEnumerator1
    Life Cycle Mode=::NewEnumerator1
    Loop Behavior=::NewEnumerator1
    Loop Duration Mode=::NewEnumerator1
    UseLoopDelay=false
   )
   &SpawnBurst_Instantaneous(Spawn Count=1, Use Spawn Probability=false)
  ]
  particleSpawnScript [
   &InitializeParticle(
    Lifetime=5
    Mesh Scale=&VectorFromFloat(Value=1.2)
    Module.Write Color=true
    Module.Write Lifetime=false
    Module.Write Mass=true
    Module.Write Material Random=true
    Module.Write Position=true
    Module.Write Scale=true
    Module.Write SpriteRotation=false
    Module.Write SpriteSize=true
    Position=&SimulationPosition()
    Sprite Size=<5 5>
   )
   /SetVariables(Particles.CurrentVel:Vector3f=$Engine.Owner.SystemXAxis) {isSetParameters=true}
  ]
  particleUpdateScript [
   &ParticleState(Kill Particles When Lifetime Has Elapsed=false, Let Infinitely Lived Particles Die When Emitter Deactivates=true)
   null() {enabled=false, module=FN_Integrator_v2}
   null() {enabled=false, module=FN_Integrator_v2001}
   &SystemLocation(MaskSpawnGroup=false)
   &self:ScratchModule(CorrectionRate=3, IntegratedValue:NiagaraFloat=$Particles.FN_Integrator_v2.Attr, Radius=650) {module=ScratchModule}
   &DynamicMaterialParameters(
    Index 0 Param 1=&FloatFromCurve(
     FloatCurve=DataInterface() {propertyValues {Curve {keys [<at 0 0 RCIM_Cubic>, {arriveTangent=0.0784, interpMode=RCIM_Cubic, leaveTangent=0.047, tangentMode=RCTM_User, time=0.5, value=290}]}}}
     CurveIndex=$Emitter.Age
     Scale Curve=&OneMinusFloat(Float=$Particles.FN_Integrator_v2001.Attr)
    )
    Index 0 Param 2=&FloatFromCurve(
     FloatCurve=DataInterface() {propertyValues {Curve {keys [
         {arriveTangent=1.5892, interpMode=RCIM_Cubic, tangentMode=RCTM_User, time=0.1942, value=0.2}
         {interpMode=RCIM_Cubic, leaveTangent=-0.0608, time=1.5, value=2.5}
        ]}}}
     CurveIndex=$Emitter.Age
     Scale Curve=&OneMinusFloat(Float=$Particles.FN_Integrator_v2001.Attr)
    )
    Module.Param0WriteEnabled=true
    Module.Param1WriteEnabled=true
    Module.Param2WriteEnabled=false
    Module.Param3WriteEnabled=false
   )
   null() {enabled=false, module=FN_MeshRenderer_v2}
   &KillParticles(Kill Particles=$User.Despawn) {enabled=false}
  ]
  rendererClasses [&NiagaraMeshRendererProperties]
  renderers [
   {rendererClass=&NiagaraMeshRendererProperties, rendererIndex=0}
   {rendererClass=&NiagaraMeshRendererProperties, rendererIndex=1}
  ]
 }
 WindMeshBurt {
  bEnabled=false
  emitterUpdateScript [
   &EmitterState(
    Enable Distance Culling=true
    Life Cycle Mode=::NewEnumerator1
    Loop Behavior=::NewEnumerator1
    Loop Duration=1
    Max Distance Response=::KillAfterParticlesFinish
    MaxDistance=8000
    Module.bUseMaxDistance=true
    Module.bUseMinDistance=false
    Scalability Mode=::NewEnumerator1
    UseLoopDelay=false
   )
   &SpawnBurst_Instantaneous(Spawn Count=1, Spawn Time=0.125, Use Spawn Probability=false)
  ]
  particleSpawnScript [&InitializeParticle(
    Color=<rgba 1 1 1 1.25>
    Lifetime=0.205
    Mesh Scale=&VectorFromFloat(Value=3)
    Module.Write Color=true
    Module.Write Lifetime=true
    Module.Write Mass=true
    Module.Write Material Random=true
    Module.Write Position=true
    Module.Write Scale=true
    Module.Write SpriteRotation=false
    Module.Write SpriteSize=false
    Position=&SimulationPosition()
    Sprite Size=<25 25>
   )]
  particleUpdateScript [
   &ParticleState()
   &ScaleColor(
    Scale Alpha=&FloatFromCurve(FloatCurve=DataInterface() {propertyValues {Curve {keys [<at 0.2946 1>,<at 1 0>]}}})
    Scale RGB=&MakeVectorFromLinearColorRGB(Color=<rgba 1 1 1 1>)
    ScaleA=true
    ScaleRGB=true
   )
   &ScaleMeshSize(Scale Factor=&VectorFromFloat(Value=&FloatFromCurve(FloatCurve=DataInterface() {propertyValues {Curve {keys [{arriveTangent=2.5191, interpMode=RCIM_Cubic, leaveTangent=2.5191, tangentMode=RCTM_User, time=0, value=0}, <at 1 1>]}}})))
  ]
  rendererClasses [&NiagaraMeshRendererProperties]
  renderers [{rendererClass=&NiagaraMeshRendererProperties, rendererIndex=0}]
 }
}
summary {
 systemName=NS_Hightower_Grape_Ball
 emitters [
  {bEnabled=false, emitterName=Rolly_Ball, rendererClasses [&NiagaraMeshRendererProperties]}
  {bEnabled=false, emitterName=Floaty_Bits, rendererClasses [&NiagaraSpriteRendererProperties]}
  {bEnabled=false, emitterName=WindMeshBurt, rendererClasses [&NiagaraMeshRendererProperties]}
  {bEnabled=false, emitterName=DirectionalBurstSpikes, rendererClasses [&NiagaraSpriteRendererProperties]}
  {bEnabled=false, emitterName=Floaty_Bits_Burst, rendererClasses [&NiagaraSpriteRendererProperties]}
 ]
 userVariables [
  User.Despawn {type=NiagaraBool}
  User.Speed {type=NiagaraFloat, defaultValue=0, description=""}
 ]
}
