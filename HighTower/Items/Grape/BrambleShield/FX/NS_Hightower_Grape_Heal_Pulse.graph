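# Niagara one-shot pulse effect, named as the heal pulse for the Grape BrambleShield item; all four emitters are disabled here.
# CPU and GPU variants burst gold sprites outward, slowed by drag and scattered by random gravity, with flare size and color driving attributes.
# A shockwave mesh expands and fades alongside a growing light shifting red to green.
# A wind mesh burst also expands and fades, turning yellow-orange to green.
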NiagaraSystem /HighTower/Items/Grape/BrambleShield/FX/NS_Hightower_Grape_Heal_Pulse
dependencies {
 usedDataInterfaces [&NiagaraDataInterfaceCurve, &NiagaraDataInterfaceVectorCurve, &NiagaraDataInterfaceColorCurve]
 usedDynamicInputs []
 usedModules [&SystemState]
}
emitters {
 Floaty_Bits_Burst_CPU {
  bEnabled=false
  emitterUpdateScript [
   &EmitterState(
    Enable Distance Culling=true
    Life Cycle Mode=::NewEnumerator1
    Loop Behavior=::NewEnumerator1
    Loop Duration=1
    MaxDistance=4500
    Module.bUseMaxDistance=true
    Module.bUseMinDistance=false
    Reset Age On Awaken=true
    Scalability Mode=::NewEnumerator1
    UseLoopDelay=false
   )
   &SpawnBurst_Instantaneous(Spawn Count=35, Use Spawn Probability=false)
  ]
  particleSpawnScript [
   /Niagara/Modules/Spawn/Initialization/InitializeParticle(
    Color=<rgba 1 0.8419 0.0781 0.5>
    Lifetime=&UniformRangedFloat(Minimum=1.25, Maximum=3)
    Module.Write Color=true
    Module.Write Lifetime=true
    Module.Write Mass=true
    Module.Write Material Random=true
    Module.Write Position=false
    Module.Write Scale=false
    Module.Write SpriteRotation=true
    Module.Write SpriteSize=true
    Position=&SimulationPosition()
    Sprite Rotation=&UniformRangedFloat(Minimum=-20, Maximum=20)
    Sprite Size=&Vector2DFromFloat(Value=&UniformRangedFloat(Minimum=6, Maximum=15))
   )
   &DynamicMaterialParameters(
    Index 0 Param 1=&UniformAOrBFloat(A=0)
    Module.Param0WriteEnabled=true
    Module.Param1WriteEnabled=false
    Module.Param2WriteEnabled=false
    Module.Param3WriteEnabled=false
   )
   &CylinderLocation(
    Cylinder Height=0.25
    Cylinder Radius=2
    MaskSpawnGroup=false
    Module.Enable Lathe Profile=false
    Module.Override Randomness=false
    Module.Surface Only=false
   )
   &AddVelocityFromPoint(Module.Use Falloff Distance=false, Velocity Strength=&UniformRangedFloat(Minimum=800, Maximum=2500)) {module=AddVelocityFromPoint001}
  ]
  particleUpdateScript [
   &ParticleState()
   /SetVariables(Particles.FlareColor=&MultiplyLinearColor(
     Linear Color=$Particles.Color
     Scale Factor=&MakeVector4(
      X=1
      Y=1
      Z=1
      W=&FloatFromCurve(FloatCurve=DataInterface() {propertyValues {Curve {keys [<at 0 0>,<at 0.0676 1.0023>,<at 0.4738 0.9198>,<at 1 0>]}}}, Scale Curve=0.025)
     )
    )) {isSetParameters=true, module=SetVariables_#1}
   /SetVariables(Particles.FlareSize:Vector2f=&Multiply_Vector2(
     A=$Particles.Initial.SpriteSize
     B=&Multiply_Vector2(
      A=&Vector2DFromFloat(Value=&Lerp_Float(A=2, B=2.25, Alpha=&Sine(Normalized Angle=$Emitter.Age, Period=1, Scale=1, Bias=1)))
      B=&Vector2DFromFloat(Value=&FloatFromCurve(FloatCurve=DataInterface() {propertyValues {Curve {keys [<at 0 0>,<at 0.0676 1.0023>,<at 0.4738 0.9198>,<at 1 0>]}}}))
     )
    )) {isSetParameters=true, module=SetVariables_#2}
   &Drag(
    Drag=&FloatFromCurve(FloatCurve=DataInterface() {propertyValues {Curve {keys [<at 0 1>,<at 0.4112 0.2232>,<at 1 0>]}}}, Scale Curve=7)
    Use Linear Drag=true
    Use Rotational Drag=false
   )
   &GravityForce(Gravity=&UniformRangedVector(Minimum=<-100 -100 -400>, Maximum=<100 100 400>))
   &SolveForcesAndVelocity(Clamp Velocity=false, Limit Acceleration=false)
   &ScaleSpriteSize(Scale Factor=&Vector2DFromFloat(Value=&FloatFromCurve(FloatCurve=DataInterface() {propertyValues {Curve {keys [
          {arriveTangent=31.8914, interpMode=RCIM_Cubic, leaveTangent=31.8914, tangentMode=RCTM_User, time=0, value=0}
          <at 0.0913 1 RCIM_Cubic>
          <at 1 0 RCIM_Cubic>
         ]}}})))
   &SpriteRotationRate(Rotation Rate=&Lerp_Float(A=-30, B=30, Alpha=$Particles.Initial.MaterialRandom))
  ]
  rendererClasses [&NiagaraSpriteRendererProperties]
  renderers [
   {rendererClass=&NiagaraSpriteRendererProperties}
   {rendererClass=&NiagaraSpriteRendererProperties, rendererIndex=1}
  ]
 }
 Floaty_Bits_Burst_GPU {
  bEnabled=false
  simTarget=GPUComputeSim
  emitterUpdateScript [
   &EmitterState(
    Enable Distance Culling=true
    Life Cycle Mode=::NewEnumerator1
    Loop Behavior=::NewEnumerator1
    Loop Duration=1
    MaxDistance=4500
    Module.bUseMaxDistance=true
    Module.bUseMinDistance=false
    Reset Age On Awaken=true
    Scalability Mode=::NewEnumerator1
    UseLoopDelay=false
   )
   &SpawnBurst_Instantaneous(Spawn Count=150, Spawn Time=0.05, Use Spawn Probability=false)
  ]
  particleSpawnScript [
   &InitializeParticle(
    Color=<rgba 1 0.8419 0.0781 0.5>
    Lifetime=&UniformRangedFloat(Minimum=1.25, Maximum=3)
    Module.Write Color=true
    Module.Write Lifetime=true
    Module.Write Mass=true
    Module.Write Material Random=true
    Module.Write Position=false
    Module.Write Scale=false
    Module.Write SpriteRotation=true
    Module.Write SpriteSize=true
    Position=&SimulationPosition()
    Sprite Rotation=&UniformRangedFloat(Minimum=-20, Maximum=20)
    Sprite Size=&Vector2DFromFloat(Value=&UniformRangedFloat(Minimum=6, Maximum=15))
   )
   &DynamicMaterialParameters(
    Index 0 Param 1=&UniformAOrBFloat(A=0)
    Module.Param0WriteEnabled=true
    Module.Param1WriteEnabled=false
    Module.Param2WriteEnabled=false
    Module.Param3WriteEnabled=false
   )
   &CylinderLocation(
    Cylinder Height=0.25
    Cylinder Radius=2
    MaskSpawnGroup=false
    Module.Enable Lathe Profile=false
    Module.Override Randomness=false
    Module.Surface Only=false
   )
   &AddVelocityFromPoint(Module.Use Falloff Distance=false, Velocity Strength=&UniformRangedFloat(Minimum=800, Maximum=2500)) {module=AddVelocityFromPoint001}
  ]
  particleUpdateScript [
   &ParticleState()
   /SetVariables(Particles.FlareColor=&MultiplyLinearColor(
     Linear Color=$Particles.Color
     Scale Factor=&MakeVector4(
      X=1
      Y=1
      Z=1
      W=&FloatFromCurve(FloatCurve=DataInterface() {propertyValues {Curve {keys [<at 0 0>,<at 0.0676 1.0023>,<at 0.4738 0.9198>,<at 1 0>]}}}, Scale Curve=0.025)
     )
    )) {isSetParameters=true, module=SetVariables_#1}
   /SetVariables(Particles.FlareSize:Vector2f=&Multiply_Vector2(
     A=$Particles.Initial.SpriteSize
     B=&Multiply_Vector2(
      A=&Vector2DFromFloat(Value=&Lerp_Float(A=2, B=2.25, Alpha=&Sine(Normalized Angle=$Emitter.Age, Period=1, Scale=1, Bias=1)))
      B=&Vector2DFromFloat(Value=&FloatFromCurve(FloatCurve=DataInterface() {propertyValues {Curve {keys [<at 0 0>,<at 0.0676 1.0023>,<at 0.4738 0.9198>,<at 1 0>]}}}))
     )
    )) {isSetParameters=true, module=SetVariables_#2}
   &Drag(
    Drag=&FloatFromCurve(FloatCurve=DataInterface() {propertyValues {Curve {keys [<at 0 1>,<at 0.4112 0.2232>,<at 1 0>]}}}, Scale Curve=7)
    Use Linear Drag=true
    Use Rotational Drag=false
   )
   &GravityForce(Gravity=&UniformRangedVector(Minimum=<-100 -100 -400>, Maximum=<100 100 400>))
   &SolveForcesAndVelocity(Clamp Velocity=false, Limit Acceleration=false)
   &ScaleSpriteSize(Scale Factor=&Vector2DFromFloat(Value=&FloatFromCurve(FloatCurve=DataInterface() {propertyValues {Curve {keys [
          {arriveTangent=31.8914, interpMode=RCIM_Cubic, leaveTangent=31.8914, tangentMode=RCTM_User, time=0, value=0}
          <at 0.0913 1 RCIM_Cubic>
          <at 1 0 RCIM_Cubic>
         ]}}})))
   &SpriteRotationRate(Rotation Rate=&Lerp_Float(A=-30, B=30, Alpha=$Particles.Initial.MaterialRandom))
  ]
  rendererClasses [&NiagaraSpriteRendererProperties]
  renderers [
   {rendererClass=&NiagaraSpriteRendererProperties}
   {rendererClass=&NiagaraSpriteRendererProperties, rendererIndex=1}
  ]
 }
 HT_LoopShockwave_CamF {
  bEnabled=false
  emitterUpdateScript [
   &EmitterState()
   &SpawnBurst_Instantaneous(Spawn Count=1, Use Spawn Probability=false)
  ]
  particleSpawnScript [
   &InitializeParticle(
    Color=<rgba 5 5 5 0.45>
    Lifetime=1.2
    Mesh Scale=&VectorFromFloat(Value=5)
    Module.Write Color=true
    Module.Write Lifetime=true
    Module.Write Mass=true
    Module.Write Material Random=true
    Module.Write Position=false
    Module.Write Scale=true
    Module.Write SpriteRotation=false
    Module.Write SpriteSize=false
    Position=&SimulationPosition()
   )
   /Niagara/Modules/Spawn/Orientation/V2/InitialMeshOrientation(
    Orientation Axis:Vector3f=<0 0 1>
    Orientation Vector=<0 1 0>
    Rotation=&MakeVector(X=&UniformRangedFloat(Minimum=0, Maximum=1), Y=0, Z=0)
    Use Orientation Vector=true
    Use Rotation Vector=true
   )
  ]
  particleUpdateScript [
   &ParticleState()
   &ScaleMeshSize(Scale Factor=/Niagara/DynamicInputs/ScaleByCurve/ScaleVectorByCurve(
     InitialValue:Vector3f=<1 1 1>
     ScaleCurve:NiagaraDataInterfaceVectorCurve=DataInterface() {propertyValues {
       XCurve {keys [
         {arriveTangent=5.0739, interpMode=RCIM_Cubic, leaveTangent=2.0639, tangentMode=RCTM_User, time=0, value=0}
         {interpMode=RCIM_Cubic, tangentMode=RCTM_User, time=1, value=1}
        ]}
       YCurve {keys [
         {arriveTangent=6.1055, interpMode=RCIM_Cubic, leaveTangent=2.0639, tangentMode=RCTM_User, time=0, value=0}
         {interpMode=RCIM_Cubic, tangentMode=RCTM_User, time=1, value=1}
        ]}
       ZCurve {keys [
         {arriveTangent=5.3487, interpMode=RCIM_Cubic, leaveTangent=2.0639, tangentMode=RCTM_User, time=0, value=0}
         {interpMode=RCIM_Cubic, tangentMode=RCTM_User, time=1, value=1}
        ]}
      }}
     CurveIndex:NiagaraFloat=$Particles.NormalizedAge
    ))
   &ScaleColor(
    Scale Alpha=&FloatFromCurve(
     FloatCurve=DataInterface() {propertyValues {Curve {keys [
         {interpMode=RCIM_Cubic, tangentMode=RCTM_User, time=0.0023, value=0}
         {arriveTangent=2.0824, interpMode=RCIM_Cubic, leaveTangent=0.0112, tangentMode=RCTM_Break, time=0.0552, value=0.9565}
         {arriveTangent=-0.1758, interpMode=RCIM_Cubic, leaveTangent=-0.1758, tangentMode=RCTM_User, time=1.0011, value=0.002}
        ]}}}
     Scale Curve=0.5
    )
    ScaleA=true
    ScaleRGB=true
   )
   /SetVariables(Particles.LightColor=&MakeLinearColorFromVectorAndFloat(
     "Vector (RGB)"=&VectorFromCurve(VectorCurve=DataInterface() {propertyValues {
        XCurve {keys [
          {interpMode=RCIM_Cubic, tangentMode=RCTM_Break, time=-0.0014, value=9.2729}
          {arriveTangent=-15.7559, interpMode=RCIM_Cubic, leaveTangent=-0.3065, tangentMode=RCTM_Break, time=0.3017, value=-0.1656}
          <at 1 0>
         ]}
        YCurve {keys [
          {interpMode=RCIM_Cubic, leaveTangent=30.1692, tangentMode=RCTM_Break, time=0, value=0}
          {interpMode=RCIM_Cubic, leaveTangent=-34.8929, tangentMode=RCTM_Break, time=0.2237, value=12}
          {interpMode=RCIM_Cubic, tangentMode=RCTM_Break, time=1, value=0}
         ]}
        ZCurve {keys [<at 0 0>,<at 0.0627 0>,<at 1 0>]}
       }})
     "Float (Alpha)"=1
    )) {isSetParameters=true}
   /SetVariables(Particles.LightRadius=&FloatFromCurve(FloatCurve=DataInterface() {propertyValues {Curve {keys [<at 0.0021 0.0215>,<at 0.1455 0.1865>,<at 0.9014 1>]}}}, Scale Curve=125)) {isSetParameters=true}
   &DynamicMaterialParameters(
    Index 0 Param 1=&FloatFromCurve(FloatCurve=DataInterface() {propertyValues {Curve {keys [<at 0 2.5>,<at 1 0>]}}})
    Index 0 Param 2=&FloatFromCurve(FloatCurve=DataInterface() {propertyValues {Curve {keys [<at 0 3>,<at 1 0>]}}}, Scale Curve=5)
    Index 0 Param 3=1
    Module.Param0WriteEnabled=true
    Module.Param1WriteEnabled=true
    Module.Param2WriteEnabled=true
    Module.Param3WriteEnabled=true
   )
  ]
  rendererClasses [&NiagaraMeshRendererProperties, &NiagaraLightRendererProperties]
  renderers [
   {rendererClass=&NiagaraMeshRendererProperties, rendererIndex=0}
   {rendererClass=&NiagaraLightRendererProperties, rendererIndex=1}
  ]
 }
 WindMeshBurt001 {
  bEnabled=false
  emitterUpdateScript [
   &EmitterState(
    Enable Distance Culling=true
    Life Cycle Mode=::NewEnumerator1
    Loop Behavior=::NewEnumerator1
    Loop Duration=1
    MaxDistance=5500
    Module.bUseMaxDistance=true
    Module.bUseMinDistance=false
    Scalability Mode=::NewEnumerator1
    UseLoopDelay=false
   )
   &SpawnBurst_Instantaneous(Spawn Count=1, Use Spawn Probability=false)
  ]
  particleSpawnScript [&InitializeParticle(
    Lifetime=0.6
    Mesh Scale=&VectorFromFloat(Value=12)
    Module.Write Color=true
    Module.Write Lifetime=true
    Module.Write Mass=true
    Module.Write Material Random=true
    Module.Write Position=true
    Module.Write Scale=true
    Module.Write SpriteRotation=false
    Module.Write SpriteSize=false
    Position=&SimulationPosition()
    Sprite Size=<25 25>
   )]
  particleUpdateScript [
   &ParticleState()
   &ScaleColor(
    Scale Alpha=&FloatFromCurve(FloatCurve=DataInterface() {propertyValues {Curve {keys [{interpMode=RCIM_Cubic, tangentMode=RCTM_Break, time=0.21, value=1}, <at 0.3504 0.25>, <at 1 0>]}}})
    Scale RGB=&MakeVectorFromLinearColorRGB(Color=&ScaleLinearColorByCurve(
      InitialValue=<rgba 1 1 1 1>
      ScaleCurve=DataInterface() {propertyValues {
        AlphaCurve {keys []}
        BlueCurve {keys [<at 0.3926 0>,<at 0.5626 0>]}
        GreenCurve {keys [<at 0.3926 1.9202>,<at 0.5626 1>]}
        RedCurve {keys [<at 0.3926 3>,<at 0.5626 0>]}
       }}
      CurveIndex=$Particles.NormalizedAge
     ))
    ScaleA=true
    ScaleRGB=true
   )
   &ScaleMeshSize(Scale Factor=&VectorFromFloat(Value=&FloatFromCurve(FloatCurve=DataInterface() {propertyValues {Curve {keys [
          {arriveTangent=0.3578, interpMode=RCIM_Cubic, leaveTangent=0.3578, tangentMode=RCTM_User, time=0.1572, value=0}
          {arriveTangent=1.9786, interpMode=RCIM_Cubic, leaveTangent=1.9786, tangentMode=RCTM_User, time=0.2938, value=0.159}
          {arriveTangent=0.6556, interpMode=RCIM_Cubic, leaveTangent=0.6556, tangentMode=RCTM_User, time=1, value=1}
         ]}}})))
  ]
  rendererClasses [&NiagaraMeshRendererProperties]
  renderers [{rendererClass=&NiagaraMeshRendererProperties, rendererIndex=0}]
 }
}
summary {
 systemName=NS_Hightower_Grape_Heal_Pulse
 emitters [
  {
   bEnabled=false
   emitterName=Floaty_Bits_Burst_GPU
   simTarget=GPUComputeSim
   rendererClasses [&NiagaraSpriteRendererProperties]
  }
  {
   bEnabled=false
   emitterName=HT_LoopShockwave_CamF
   rendererClasses [&NiagaraMeshRendererProperties, &NiagaraLightRendererProperties]
  }
  {bEnabled=false, emitterName=WindMeshBurt001, rendererClasses [&NiagaraMeshRendererProperties]}
  {bEnabled=false, emitterName=Floaty_Bits_Burst_CPU, rendererClasses [&NiagaraSpriteRendererProperties]}
 ]
 userVariables []
}
