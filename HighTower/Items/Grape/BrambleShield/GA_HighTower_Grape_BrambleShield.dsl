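; Gameplay ability for the HighTower Grape (Groot) Bramble Shield, which encases the player in a destructible ball that periodically heals.
; Activation applies knockback, launches the player, holsters their weapon, spawns the attached shield on the server and enlarges the capsule, retrying while pawns overlap.
; A timer applies safe-zone-dependent heals until a final heal.
; Timeout, destruction or delayed cancel run net-synced cleanup that releases the ball, restores the capsule and relaunches the player.
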(event EventOnAbilityActivated
  (bind _playerpawn (Variables|Default|GetPlayerPawn))
  ;[1 1 1] On Activated
  (|CallParentFunction)
  (Utilities|Array|Add (Variables|Default|GetActorsToIgnore) _playerpawn)
  (if (Variables|/Script/Engine.Character|GetbIsCrouched _playerpawn)
    (Character|UnCrouch _playerpawn)
    (Utilities|FlowControl|Delay 0.1)
    (else
      _))
  (bind _returnvalue (Zipline|IsZiplining _playerpawn))
  (if _returnvalue
    (Zipline|EndZiplining _playerpawn false false)
    (else
      _))
  (CallFunction|ApplyEffecttoSelf))

(event Custom|SetupBinds
  ;[1 1 1] Set up additional Binds for ending the ability
  (Ability|Tasks|WaitGameplayEvent "(TagName=\"Abilities.Generic.Rift.Trigger\")"
    (:then
      (Ability|Tasks|WaitGameplayTagAdd "(TagName=\"Gameplay.Action.Player.DanceStunned\")"
        (:Added
          _)))
    (:EventReceived
      _))
  (Ability|EndAbility))

(event Custom|KnockBack
  ;[1 1 1] Knockback - Execute activate cue, target and apply knockback
  (Ability|ExecuteGameplayCueOnOwner (Variables|Default|GetGC_Activate))
  (bind (_asynctaskproxy _targetdata) (Ability|Tasks|DoTargetSelection "None" (Variables|Default|GetTag_KnockbackTargeting) true)
    (:Targeted
      (Ability|ApplyGameplayEffectToTarget _targetdata (Variables|Default|GetGE_Knockback) 0))))

(event Custom|ShieldCancelled
  ;[0.1094 0.07261 0.005] When shield is cancelled, end the ability
  (if (Variables|DestructionandCancelling|GetCanBeCancelled)
    (Utilities|FlowControl|DoOnce false
      (:Completed
        (CallFunction|OnShieldDestroyedTimedOutorCancelled)
        (Variables|DestructionandCancelling|SetCanBeCancelled false)))))

(event Custom|ResetCancelDoOnce
  ;[0.1094 0.07261 0.005] When shield is cancelled, end the ability
  (Utilities|FlowControl|DoOnce false
    (:Completed
      (CallFunction|OnShieldDestroyedTimedOutorCancelled)
      (Variables|DestructionandCancelling|SetCanBeCancelled false))))

(event EventApplyEffectToSelf
  ;[1 1 1] After Effect is applied, spawn shield, setup heal, and set up cancel
  (|CallParentFunction)
  (Ability|Tasks|WaitGameplayEvent (Variables|Default|GetEvent_ForceEnd)
    (:then
      (CallFunction|KnockBack)
      (CallFunction|SpawnShield)
      (CallFunction|SetUpHeal)
      (Utilities|Time|SetTimerbyEvent (delegate Custom|SetUpCanBeCancelledTimer) (Variables|DestructionandCancelling|GetDelayBeforeCanBeCancelled))
      (Utilities|Time|SetTimerbyEvent (delegate Custom|HolsterWeapon) (Variables|Default|GetDelayForWeaponHolster)))
    (:EventReceived
      (CallFunction|ShieldCancelled))))

(event Custom|SetUpHeal
  ;[0 1 0] Set Up Heal - Start the heal timer
  (bind _returnvalue (Utilities|Time|SetTimerbyEvent (EventDispatchers|CreateEvent :!event Custom|Heal) (ScalableFloat|GetValueatLevel (Variables|Healing|GetRow_HealInterval)) true))
  (Variables|Healing|SetTimerHandle_Heal _returnvalue))

(event Custom|SetUpCanBeCancelledTimer
  ;[1 1 1] After Effect is applied, spawn shield, setup heal, and set up cancel
  (Variables|DestructionandCancelling|SetCanBeCancelled true))

(event Custom|HolsterWeapon
  ;[1 1 1] After Effect is applied, spawn shield, setup heal, and set up cancel
  (Weapon|SetHolsterWeapon (Variables|Default|GetPlayerPawn) 0 true true "Holstered for GrootBall")
  (CallFunction|ResetUnholsterWeapon))

(event Custom|SpawnShield
  ;[1 1 1] Spawn Shield - Spawn the shield and attach it to the player
  (Utilities|FlowControl|Sequence
    (:then_0
      (if (Networking|IsServer)
        (Utilities|IsValid (Variables|BrambleShield|GetBrambleShield)
          (:Is%20Valid
            (CallFunction|OnShieldDestroyedTimedOutorCancelled))
          (:Is%20Not%20Valid
            (bind _returnvalue (Game|SpawnActorfromClass "/HighTower/Items/Grape/BrambleShield/BGA_HighTower_Grape_BrambleShield.BGA_HighTower_Grape_BrambleShield_C" (CallFunction|GetShieldSpawnTransform) "AlwaysSpawn" "OverrideRootScale" 0 0 (Variables|Default|GetPlayerPawn)))
            (bind _output_get (Variables|BrambleShield|SetBrambleShield _returnvalue))
            (EventDispatchers|BindEventtoOnDied _output_get (EventDispatchers|CreateEvent :!event Custom|BrambleDeath))
            (Transformation|AttachActorToActor _output_get (Variables|Default|GetPlayerPawn) "None" "SnapToTarget" "KeepWorld" "KeepRelative" true)
            (CallFunction|SetupBinds)
            (Fortnite|ApplyGameplayEffectsToActor (Variables|BrambleShield|GetBrambleShield) (Variables|Default|GetGEs_ToApplyToBrambleShield) (Transformation|GetActorLocation (Variables|Default|GetPlayerPawn)) false 0 (Ability|GetAbilitySystemComponent (Variables|Default|GetPlayerPawn)))
            (Utilities|Array|Add (Variables|Default|GetActorsToIgnore) (Variables|BrambleShield|GetBrambleShield))))))
    (:then_1
      (CallFunction|TryEnlargePawnCapsule))
    (:then_2
      (CallFunction|JumpAndImpulseCharacter true (ScalableFloat|GetValueatLevel (Variables|Default|GetLaunchHorizSpeedStart)) (ScalableFloat|GetValueatLevel (Variables|Default|GetLaunchVertSpeedStart))))))

(event Custom|Heal
  (bind _playerpawn (Variables|Default|GetPlayerPawn))
  (bind _tag_insafezone (Variables|Healing|GetTag_InSafeZone))
  (bind _case_0_true (GameplayTags|TagContainer|CompareTagAssetInterfacetoOtherTags _tag_insafezone _playerpawn))
  (bind _ge_heal (Variables|Healing|GetGE_Heal))
  (bind _ge_healinstorm (Variables|Healing|GetGE_HealInStorm))
  (bind _returnvalue (select _case_0_true _ge_heal _ge_healinstorm))
  ;[0 1 0] Heal - Target and Heal if the heal is on the last heal call Final Heal Instead
  (if (CallFunction|OnLastHeal)
    (CallFunction|FinalHeal)
    (else
      (Ability|ExecuteGameplayCueOnOwner (Variables|Healing|GetGC_Heal))
      (bind (_asynctaskproxy _targetdata) (Ability|Tasks|DoTargetSelection "None" (Variables|Healing|GetTag_HealTargeting) true)
        (:then
          (Ability|ApplyGameplayEffectToOwner _returnvalue 0)
          (Math|Integer|IncrementInt (Variables|Healing|GetHealCount)))
        (:Targeted
          (Ability|ApplyGameplayEffectToTarget _targetdata _returnvalue 0))))))

(event Custom|FinalHeal
  (bind _playerpawn (Variables|Default|GetPlayerPawn))
  (bind _tag_insafezone (Variables|Healing|GetTag_InSafeZone))
  (bind _case_0_true (GameplayTags|TagContainer|CompareTagAssetInterfacetoOtherTags _tag_insafezone _playerpawn))
  (bind _ge_finalheal (Variables|Healing|GetGE_FinalHeal))
  (bind _ge_finalhealinstorm (Variables|Healing|GetGE_FinalHealInStorm))
  (bind _returnvalue (select _case_0_true _ge_finalheal _ge_finalhealinstorm))
  ;[0 1 0] Final Heal - Do the final heal then end ability
  (Utilities|FlowControl|Sequence
    (:then_0
      (Ability|ExecuteGameplayCueOnOwner (Variables|Healing|GetGC_FinalHeal))
      (bind (_asynctaskproxy _targetdata) (Ability|Tasks|DoTargetSelection "None" (Variables|Healing|GetTag_FinalHealTargeting) true)
        (:then
          (Ability|ApplyGameplayEffectToOwner _returnvalue 0))
        (:Targeted
          (Ability|ApplyGameplayEffectToTarget _targetdata _returnvalue 0))))
    (:then_1
      (Utilities|Time|ClearandInvalidateTimerbyHandle (Variables|Healing|GetTimerHandle_Heal))
      (bind _returnvalue_2 (Utilities|Time|SetTimerbyEvent (delegate Custom|PostFinalHealTimerExpiredStartFX) (Variables|Healing|GetDelay_PostFinalHeal)))
      (Variables|Healing|SetTimerHandle_Heal _returnvalue_2))))

(event Custom|TryEnlargePawnCapsule
  (bind _playerpawn (Variables|Default|GetPlayerPawn))
  (bind _capsulecomponent (Variables|/Script/Engine.Character|GetCapsuleComponent _playerpawn))
  ;[1 1 1] Enlarge Capsule
  (Variables|Default|SetHitAPawn false)
  (Utilities|FlowControl|Sequence
    (:then_0
      (if (Variables|Default|GetCapsuleIsEnlarged)
        (bind _outhits (Collision|MultiSphereTraceForObjects (Transformation|GetActorLocation (Variables|Default|GetPlayerPawn)) (+ (Transformation|GetActorLocation (Variables|Default|GetPlayerPawn)) "0, 0,1.000000") (+ (Variables|Default|GetBrambleRadiusForPawnCapsule) 4.0) (Variables|Default|GetObjectTypes) false (Variables|Default|GetActorsToIgnore)))
        (for _array_element _outhits
          (Utilities|Casting|CastToFortPlayerPawn (.HitActor (Collision|BreakHitResult _array_element)))
          (Variables|Default|SetHitAPawn true))))
    (:then_1
      ;;[1 1 1] If we hit a pawn, try again in a little bit
      (if (Variables|Default|GetHitAPawn)
        (Utilities|Time|SetTimerbyEvent (delegate Custom|TryEnlargePawnCapsule) 0.2)
        (else
          (Variables|Default|SetCachedUnscaledPawnCapsuleRadius (Components|Capsule|GetUnscaledCapsuleRadius _capsulecomponent))
          (Variables|Default|SetCachedCapsuleHalfHeight (Components|Capsule|GetUnscaledCapsuleHalfHeight _capsulecomponent))
          (Variables|Default|SetCapsuleIsEnlarged true)
          (Components|Capsule|SetCapsuleRadius _capsulecomponent (Variables|Default|GetBrambleRadiusForPawnCapsule) true)
          (Utilities|IsServer
            (:true
              (bind _asplayer_pawn_athena_generic (Utilities|Casting|CastToPlayerPawn_Athena_Generic _playerpawn))
              (Variables|/Game/Athena/PlayerPawn_Athena_Generic.PlayerPawn_Athena_Generic_C|SetReppedCapsuleExtraHalfHeight (- (Components|Capsule|GetUnscaledCapsuleRadius (Variables|/Script/Engine.Character|GetCapsuleComponent (Variables|Default|GetPlayerPawn))) (Variables|Default|GetCachedCapsuleHalfHeight)) _asplayer_pawn_athena_generic))))))))

(event Custom|PostFinalHealTimerExpiredStartFX
  ;[0 1 0] Final Heal - Do the final heal then end ability
  (Class|BGAHighTowerGrapeBrambleShield|EndBramble (Variables|BrambleShield|GetBrambleShield) true)
  (bind _returnvalue (Utilities|Time|SetTimerbyEvent (delegate Custom|PostFinalHealTimerExpiredExitBall) (Variables|Healing|GetTimeToReceedBeforeExit)))
  (Variables|Healing|SetTimerHandle_Heal _returnvalue))

(event Custom|PostFinalHealTimerExpiredExitBall
  ;[0 1 0] Final Heal - Do the final heal then end ability
  (CallFunction|OnShieldDestroyedTimedOutorCancelled))

(event Custom|RestorePawnCapsule
  (bind _playerpawn (Variables|Default|GetPlayerPawn))
  (bind _capsulecomponent (Variables|/Script/Engine.Character|GetCapsuleComponent _playerpawn))
  ;[1 1 1] Restore Capsule to its normal size
  (if (Variables|Default|GetCapsuleIsEnlarged)
    (Components|Capsule|SetCapsuleRadius _capsulecomponent (Variables|Default|GetCachedUnscaledPawnCapsuleRadius) true)
    (Components|Capsule|SetCapsuleHalfHeight _capsulecomponent (Variables|Default|GetCachedCapsuleHalfHeight) true)
    (Utilities|IsServer
      (:true
        (bind _asplayer_pawn_athena_generic (Utilities|Casting|CastToPlayerPawn_Athena_Generic _playerpawn))
        (Variables|/Game/Athena/PlayerPawn_Athena_Generic.PlayerPawn_Athena_Generic_C|SetReppedCapsuleExtraHalfHeight 0.0 _asplayer_pawn_athena_generic)))))

(event Custom|BrambleDeath
  (params (DamagedActor :type Actor%20Object%20Reference) (Damage :type Float%20%28single-precision%29) (InstigatedBy :type Controller%20Object%20Reference) (DamageCauser :type Actor%20Object%20Reference) (HitLocation :type Vector) (FHitComponent :type Primitive%20Component%20Object%20Reference) (BoneName :type Name) (Momentum :type Vector))
  ;[1 1 1] Bramble Death - End Ability if shield dies
  (Development|PrintString "Bramble Death (Ability)" (Variables|Debug|GetDebugLogBrambleShield) (Variables|Debug|GetDebugLogBrambleShield))
  (Variables|DestructionandCancelling|SetDestroyed true)
  (CallFunction|OnShieldDestroyedTimedOutorCancelled)
  (CallFunction|ShieldDestoyedClient))

(event Custom|EndCalledCleanUp
  (bind _brambleshield (Variables|BrambleShield|GetBrambleShield))
  (bind _brambleshield_2 (Variables|BrambleShield|GetBrambleShield))
  ;[1 1 1] On shield destroyed or cancelled
  (Development|PrintString "On Shield Cancelled Timed Out Or Destroyed" (Variables|Debug|GetDebugLogBrambleShield) (Variables|Debug|GetDebugLogBrambleShield))
  (if (not (Variables|DestructionandCancelling|GetDestroyed))
    (Utilities|IsValid _brambleshield_2
      (:Is%20Valid
        (Class|BGAHighTowerGrapeBrambleShield|EndBramble _brambleshield_2 false))
      (:Is%20Not%20Valid
        _))
    (Utilities|Time|ClearandInvalidateTimerbyHandle (Variables|Healing|GetTimerHandle_Heal))
    (if (not (Variables|Debug|GetDebugLogBrambleShield))
      (Ability|CommitAbility))
    (CallFunction|RestorePawnCapsule)
    ;[1 1 1] On the end unattach the ball and impulse it away from you.
    (Transformation|DetachFromActor _brambleshield "KeepRelative" "KeepRelative" "KeepRelative")
    (Class|BGAHighTowerGrapeBrambleShield|TurnonGravity _brambleshield (+ (* (Math|Vector|MakeVector (.x (Pawn|Components|CharacterMovement|GetLastUpdateVelocity (Variables|/Script/Engine.Character|GetCharacterMovement (Variables|Default|GetPlayerPawn)))) (.y (Pawn|Components|CharacterMovement|GetLastUpdateVelocity (Variables|/Script/Engine.Character|GetCharacterMovement (Variables|Default|GetPlayerPawn))))) (Variables|Default|GetBallVelMult)) (Math|Vector|SelectVector (Variables|Default|GetBallEndAdditionVelOnGround) (Variables|Default|GetBallEndAdditionVelInAir) (AI|Components|NavMovement|IsMovingonGround (Variables|/Script/Engine.Character|GetCharacterMovement (Variables|Default|GetPlayerPawn))))))
    (if (not (Variables|Default|GetEndAbilityHasBeenCalled))
      (Utilities|FlowControl|Delay (Variables|Default|GetPostRemovalLaunch))
      ;[1 1 1] Jump Character
      (CallFunction|JumpAndImpulseCharacter false (ScalableFloat|GetValueatLevel (Variables|Default|GetLaunchHorizSpeedEnd)) (ScalableFloat|GetValueatLevel (Variables|Default|GetLaunchVertSpeedEnd)) true)
      (CallFunction|UnholsterWeapon)
      (Utilities|FlowControl|Delay 0.01) ; Frame delay to ensure we remove after launching
    )
    (Ability|RemoveGameplayEffectFromOwnerWithGrantedTags (Variables|Default|GetT_BrambleActive))
    (if (not (Variables|Default|GetEndAbilityHasBeenCalled))
      ;[1 1 1] Delay and end
      (Utilities|FlowControl|Delay (Variables|Default|GetPostEndCleanUpDelay)) ; Wait a moment before ending so that everything has time to resolve correctly
      (Ability|EndAbility))))

(event Custom|ShieldDestoyedClient
  (bind _brambleshield (Variables|BrambleShield|GetBrambleShield))
  (bind _brambleshield_2 (Variables|BrambleShield|GetBrambleShield))
  ;[1 1 1] On shield destroyed or cancelled
  (Utilities|FlowControl|DoOnce false
    (:Completed
      (Ability|Tasks|WaitNetSync "BothWait"
        (:OnSync
          (Development|PrintString "On Shield Cancelled Timed Out Or Destroyed" (Variables|Debug|GetDebugLogBrambleShield) (Variables|Debug|GetDebugLogBrambleShield))
          (if (not (Variables|DestructionandCancelling|GetDestroyed))
            (Utilities|IsValid _brambleshield_2
              (:Is%20Valid
                (Class|BGAHighTowerGrapeBrambleShield|EndBramble _brambleshield_2 false))
              (:Is%20Not%20Valid
                _))
            (Utilities|Time|ClearandInvalidateTimerbyHandle (Variables|Healing|GetTimerHandle_Heal))
            (if (not (Variables|Debug|GetDebugLogBrambleShield))
              (Ability|CommitAbility))
            (CallFunction|RestorePawnCapsule)
            ;[1 1 1] On the end unattach the ball and impulse it away from you.
            (Transformation|DetachFromActor _brambleshield "KeepRelative" "KeepRelative" "KeepRelative")
            (Class|BGAHighTowerGrapeBrambleShield|TurnonGravity _brambleshield (+ (* (Math|Vector|MakeVector (.x (Pawn|Components|CharacterMovement|GetLastUpdateVelocity (Variables|/Script/Engine.Character|GetCharacterMovement (Variables|Default|GetPlayerPawn)))) (.y (Pawn|Components|CharacterMovement|GetLastUpdateVelocity (Variables|/Script/Engine.Character|GetCharacterMovement (Variables|Default|GetPlayerPawn))))) (Variables|Default|GetBallVelMult)) (Math|Vector|SelectVector (Variables|Default|GetBallEndAdditionVelOnGround) (Variables|Default|GetBallEndAdditionVelInAir) (AI|Components|NavMovement|IsMovingonGround (Variables|/Script/Engine.Character|GetCharacterMovement (Variables|Default|GetPlayerPawn))))))
            (if (not (Variables|Default|GetEndAbilityHasBeenCalled))
              (Utilities|FlowControl|Delay (Variables|Default|GetPostRemovalLaunch))
              ;[1 1 1] Jump Character
              (CallFunction|JumpAndImpulseCharacter false (ScalableFloat|GetValueatLevel (Variables|Default|GetLaunchHorizSpeedEnd)) (ScalableFloat|GetValueatLevel (Variables|Default|GetLaunchVertSpeedEnd)) true)
              (CallFunction|UnholsterWeapon)
              (Utilities|FlowControl|Delay 0.01) ; Frame delay to ensure we remove after launching
            )
            (Ability|RemoveGameplayEffectFromOwnerWithGrantedTags (Variables|Default|GetT_BrambleActive))
            (if (not (Variables|Default|GetEndAbilityHasBeenCalled))
              ;[1 1 1] Delay and end
              (Utilities|FlowControl|Delay (Variables|Default|GetPostEndCleanUpDelay)) ; Wait a moment before ending so that everything has time to resolve correctly
              (Ability|EndAbility))))) ; Ensure the client and server sync up here.
    )))

(event Custom|OnShieldDestroyedTimedOutOrCancelled
  (bind _brambleshield (Variables|BrambleShield|GetBrambleShield))
  (bind _brambleshield_2 (Variables|BrambleShield|GetBrambleShield))
  ;[1 1 1] On shield destroyed or cancelled
  (Utilities|FlowControl|DoOnce false
    (:Completed
      (Ability|Tasks|WaitNetSync "BothWait"
        (:OnSync
          (Development|PrintString "On Shield Cancelled Timed Out Or Destroyed" (Variables|Debug|GetDebugLogBrambleShield) (Variables|Debug|GetDebugLogBrambleShield))
          (if (not (Variables|DestructionandCancelling|GetDestroyed))
            (Utilities|IsValid _brambleshield_2
              (:Is%20Valid
                (Class|BGAHighTowerGrapeBrambleShield|EndBramble _brambleshield_2 false))
              (:Is%20Not%20Valid
                _))
            (Utilities|Time|ClearandInvalidateTimerbyHandle (Variables|Healing|GetTimerHandle_Heal))
            (if (not (Variables|Debug|GetDebugLogBrambleShield))
              (Ability|CommitAbility))
            (CallFunction|RestorePawnCapsule)
            ;[1 1 1] On the end unattach the ball and impulse it away from you.
            (Transformation|DetachFromActor _brambleshield "KeepRelative" "KeepRelative" "KeepRelative")
            (Class|BGAHighTowerGrapeBrambleShield|TurnonGravity _brambleshield (+ (* (Math|Vector|MakeVector (.x (Pawn|Components|CharacterMovement|GetLastUpdateVelocity (Variables|/Script/Engine.Character|GetCharacterMovement (Variables|Default|GetPlayerPawn)))) (.y (Pawn|Components|CharacterMovement|GetLastUpdateVelocity (Variables|/Script/Engine.Character|GetCharacterMovement (Variables|Default|GetPlayerPawn))))) (Variables|Default|GetBallVelMult)) (Math|Vector|SelectVector (Variables|Default|GetBallEndAdditionVelOnGround) (Variables|Default|GetBallEndAdditionVelInAir) (AI|Components|NavMovement|IsMovingonGround (Variables|/Script/Engine.Character|GetCharacterMovement (Variables|Default|GetPlayerPawn))))))
            (if (not (Variables|Default|GetEndAbilityHasBeenCalled))
              (Utilities|FlowControl|Delay (Variables|Default|GetPostRemovalLaunch))
              ;[1 1 1] Jump Character
              (CallFunction|JumpAndImpulseCharacter false (ScalableFloat|GetValueatLevel (Variables|Default|GetLaunchHorizSpeedEnd)) (ScalableFloat|GetValueatLevel (Variables|Default|GetLaunchVertSpeedEnd)) true)
              (CallFunction|UnholsterWeapon)
              (Utilities|FlowControl|Delay 0.01) ; Frame delay to ensure we remove after launching
            )
            (Ability|RemoveGameplayEffectFromOwnerWithGrantedTags (Variables|Default|GetT_BrambleActive))
            (if (not (Variables|Default|GetEndAbilityHasBeenCalled))
              ;[1 1 1] Delay and end
              (Utilities|FlowControl|Delay (Variables|Default|GetPostEndCleanUpDelay)) ; Wait a moment before ending so that everything has time to resolve correctly
              (Ability|EndAbility))))) ; Ensure the client and server sync up here.
    )))

(event Custom|ResetCanBeDestroyedTimedOutOrCancelled
  (bind _brambleshield (Variables|BrambleShield|GetBrambleShield))
  (bind _brambleshield_2 (Variables|BrambleShield|GetBrambleShield))
  ;[1 1 1] On shield destroyed or cancelled
  (Utilities|FlowControl|DoOnce false
    (:Completed
      (Ability|Tasks|WaitNetSync "BothWait"
        (:OnSync
          (Development|PrintString "On Shield Cancelled Timed Out Or Destroyed" (Variables|Debug|GetDebugLogBrambleShield) (Variables|Debug|GetDebugLogBrambleShield))
          (if (not (Variables|DestructionandCancelling|GetDestroyed))
            (Utilities|IsValid _brambleshield_2
              (:Is%20Valid
                (Class|BGAHighTowerGrapeBrambleShield|EndBramble _brambleshield_2 false))
              (:Is%20Not%20Valid
                _))
            (Utilities|Time|ClearandInvalidateTimerbyHandle (Variables|Healing|GetTimerHandle_Heal))
            (if (not (Variables|Debug|GetDebugLogBrambleShield))
              (Ability|CommitAbility))
            (CallFunction|RestorePawnCapsule)
            ;[1 1 1] On the end unattach the ball and impulse it away from you.
            (Transformation|DetachFromActor _brambleshield "KeepRelative" "KeepRelative" "KeepRelative")
            (Class|BGAHighTowerGrapeBrambleShield|TurnonGravity _brambleshield (+ (* (Math|Vector|MakeVector (.x (Pawn|Components|CharacterMovement|GetLastUpdateVelocity (Variables|/Script/Engine.Character|GetCharacterMovement (Variables|Default|GetPlayerPawn)))) (.y (Pawn|Components|CharacterMovement|GetLastUpdateVelocity (Variables|/Script/Engine.Character|GetCharacterMovement (Variables|Default|GetPlayerPawn))))) (Variables|Default|GetBallVelMult)) (Math|Vector|SelectVector (Variables|Default|GetBallEndAdditionVelOnGround) (Variables|Default|GetBallEndAdditionVelInAir) (AI|Components|NavMovement|IsMovingonGround (Variables|/Script/Engine.Character|GetCharacterMovement (Variables|Default|GetPlayerPawn))))))
            (if (not (Variables|Default|GetEndAbilityHasBeenCalled))
              (Utilities|FlowControl|Delay (Variables|Default|GetPostRemovalLaunch))
              ;[1 1 1] Jump Character
              (CallFunction|JumpAndImpulseCharacter false (ScalableFloat|GetValueatLevel (Variables|Default|GetLaunchHorizSpeedEnd)) (ScalableFloat|GetValueatLevel (Variables|Default|GetLaunchVertSpeedEnd)) true)
              (CallFunction|UnholsterWeapon)
              (Utilities|FlowControl|Delay 0.01) ; Frame delay to ensure we remove after launching
            )
            (Ability|RemoveGameplayEffectFromOwnerWithGrantedTags (Variables|Default|GetT_BrambleActive))
            (if (not (Variables|Default|GetEndAbilityHasBeenCalled))
              ;[1 1 1] Delay and end
              (Utilities|FlowControl|Delay (Variables|Default|GetPostEndCleanUpDelay)) ; Wait a moment before ending so that everything has time to resolve correctly
              (Ability|EndAbility))))) ; Ensure the client and server sync up here.
    )))

(event Custom|Jump%20and%20Impulse%20Character
  (params (UseAcceleration :type Boolean) (HorizSpeed :type Float%20%28double-precision%29) (VertSpeed :type Float%20%28double-precision%29) (IsFinishingAbility :type Boolean))
  (bind (_returnvalue_x _returnvalue_y) (Math|Vector|BreakVector (Pawn|Components|CharacterMovement|GetCurrentAcceleration (Variables|/Script/Engine.Character|GetCharacterMovement (Variables|Default|GetPlayerPawn)))))
  (bind _playerpawn (Variables|Default|GetPlayerPawn))
  (bind (_velocity_x _velocity_y) (Math|Vector|BreakVector (Variables|/Script/Engine.MovementComponent|GetVelocity_X (Variables|/Script/Engine.Character|GetCharacterMovement _playerpawn))))
  ;[1 1 1] JumpAndImpulseCharacter
  (Networking|ForceNetUpdate _playerpawn)
  (Character|LaunchCharacter (Variables|Default|GetPlayerPawn) (+ (Math|Vector|SelectVector (* (Math|Vector|Normalize (Math|Vector|SelectVector (Math|Vector|SelectVector (Transformation|GetActorForwardVector _playerpawn) (Math|Vector|MakeVector _returnvalue_x _returnvalue_y) (Math|Float|NearlyEqual(Float) (Math|Vector|VectorLength (Math|Vector|MakeVector _returnvalue_x _returnvalue_y)) 0.0 0.0001)) (Math|Vector|SelectVector (Transformation|GetActorForwardVector _playerpawn) "0, 0, 0" (Math|Float|NearlyEqual(Float) (Math|Vector|VectorLength (Math|Vector|MakeVector _velocity_x _velocity_y)) 0.0 0.0001)) UseAcceleration)) HorizSpeed) (+ (* (Math|Vector|Normalize (Math|Vector|SelectVector (Math|Vector|SelectVector (Transformation|GetActorForwardVector _playerpawn) (Math|Vector|MakeVector _returnvalue_x _returnvalue_y) (Math|Float|NearlyEqual(Float) (Math|Vector|VectorLength (Math|Vector|MakeVector _returnvalue_x _returnvalue_y)) 0.0 0.0001)) (Math|Vector|SelectVector (Transformation|GetActorForwardVector _playerpawn) "0, 0, 0" (Math|Float|NearlyEqual(Float) (Math|Vector|VectorLength (Math|Vector|MakeVector _velocity_x _velocity_y)) 0.0 0.0001)) UseAcceleration)) HorizSpeed) (* (Math|Vector|MakeVector _velocity_x _velocity_y) (Variables|Default|GetEndJumpVelMaintained))) UseAcceleration) (Math|Vector|MakeVector 0.0 0.0 VertSpeed)) true true))

(event Ability|EventOnEndAbility
  (params (bWasCancelled :type Boolean))
  ;[1 1 1] On End Ability
  (|CallParentFunction bWasCancelled)
  (CallFunction|UnholsterWeapon)
  (Variables|Default|SetEndAbilityHasBeenCalled true)
  (CallFunction|EndCalledCleanUp)
  (Ability|ApplyGameplayEffectToOwner (Variables|Default|GetGE_ImmunityToNextFallDamage)) ; Give Immunity to next fall damage
)

(event Custom|UnholsterWeapon
  ;[1 1 1] UnholsterWeapon
  (Utilities|FlowControl|DoOnce false
    (:Completed
      (if true
        (Weapon|SetHolsterWeapon (Variables|Default|GetPlayerPawn) 0 false true "Holstered for GrootBall")))))

(event Custom|ResetUnholsterWeapon
  ;[1 1 1] UnholsterWeapon
  (Utilities|FlowControl|DoOnce false
    (:Completed
      (if true
        (Weapon|SetHolsterWeapon (Variables|Default|GetPlayerPawn) 0 false true "Holstered for GrootBall")))))

(fn GetShieldSpawnTransform
  ;[1 1 1] GetShieldSpawnTransform
  (return (Math|Transform|MakeTransform (Transformation|GetActorLocation (Variables|Default|GetPlayerPawn)) (Transformation|GetActorRotation (Variables|Default|GetPlayerPawn)))))

(fn InitAbility
  ;[1 1 1] Init Ability
  (CallFunction|ResetUnholsterWeapon)
  (CallFunction|ResetCancelDoOnce)
  (CallFunction|ResetCanbeDestroyedTimedOutorCancelled)
  (Variables|DestructionAndCancelling|SetCancelled false)
  (Variables|DestructionandCancelling|SetDestroyed false)
  (Variables|DestructionandCancelling|SetCanBeCancelled false)
  (Variables|Healing|SetHealCount 0)
  (Variables|Default|SetEndAbilityHasBeenCalled false)
  (|CallParentFunction))

(fn OnLastHeal
  ;[1 1 1] Check if we have hit the max amount of heals
  (return (<= (ScalableFloat|GetValueatLevel (Variables|Healing|GetRow_MaxHealCount)) (Math|Conversions|ToFloat(Integer) (+ (Variables|Healing|GetHealCount) 1)))))
