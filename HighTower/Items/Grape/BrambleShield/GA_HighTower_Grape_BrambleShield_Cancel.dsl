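; Gameplay ability that cancels the player's active Grape Bramble Shield in HighTower.
; On activation it prints a debug message and, if the avatar actor is valid, sends it a force-end-shield gameplay event.
; It then calls EndAbility, leaving the actual shield teardown to whatever handles that event.
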(event Ability|EventActivateAbility
  (bind _returnvalue (Ability|GetAvatarActorfromActorInfo))
  ;[1 1 1] Send event to the player to kill the Bramble Shield.
  (Development|PrintString "Player Cancelled Bramble")
  (Utilities|IsValid _returnvalue
    (:Is%20Valid
      (Ability|SendGameplayEventtoActor _returnvalue (Variables|Default|GetEvent_ForceEndShield)))
    (:Is%20Not%20Valid
      _))
  (Ability|EndAbility))
