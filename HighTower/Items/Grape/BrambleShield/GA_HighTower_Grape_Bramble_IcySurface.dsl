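; Gameplay ability for the HighTower Grape Bramble Shield's icy surface, giving the player slippery movement for halfpipe-style slope riding.
; On activation it applies icy-surface, acceleration and immunity effects, widens the walkable floor angle and starts a looping slope check.
; The check compares height samples to set the acceleration effect's level, separating clear climbs, other walking and falling.
; Ending removes the effects, restores the angle, clears holstering and stops the timer.
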(event Custom|CheckSlope
  (bind _playerpawn (Variables|Default|GetPlayerPawn))
  (bind (_returnvalue_x _returnvalue_y _returnvalue_z) (Math|Vector|BreakVector (Transformation|GetActorLocation _playerpawn)))
  (bind _playerpawn_2 (Variables|Default|GetPlayerPawn))
  (bind _returnvalue (Ability|GetAbilitySystemComponent _playerpawn_2))
  (bind _accelhandle (Variables|Default|GetAccelHandle))
  (bind _oldz (Variables|Default|GetOldZ))
  ;[1 1 1] Bootleg check to see if you're moving up or down a slope
  (switch Utilities|FlowControl|Switch|SwitchonEMovementMode (Variables|/Script/Engine.CharacterMovementComponent|GetMovementMode (Variables|/Script/Engine.Character|GetCharacterMovement _playerpawn))
    (:MOVE_Walking
      (bind _output_get (Variables|Default|SetNewZ _returnvalue_z))
      ;[1 1 1] Adjust value of GE that controls acceleration and speed based on slope.  Neutral for flat ground, faster or slower for down slope and up slope respectively.
      (if (Math|Float|NearlyEqual(Float) _output_get _oldz 4.0)
        (GameplayEffects|SetActiveGameplayEffectLevel _returnvalue (Variables|Default|GetAccelHandle) 1)
        (Variables|Default|SetOldZ (Variables|Default|GetNewZ))
        (elif (>= (- _output_get 1e+01) _oldz)
          (GameplayEffects|SetActiveGameplayEffectLevel _returnvalue _accelhandle 2)
          (Variables|Default|SetOldZ (Variables|Default|GetNewZ))
          (else
            (GameplayEffects|SetActiveGameplayEffectLevel _returnvalue _accelhandle 1)))))
    (:MOVE_Falling
      (GameplayEffects|SetActiveGameplayEffectLevel _returnvalue (Variables|Default|GetAccelHandle) 0))))

(event Ability|EventActivateAbility
  (bind _returnvalue (Ability|GetAbilitySystemComponent (Variables|Default|GetPlayerPawn)))
  (bind _playerpawn (Variables|Default|GetPlayerPawn))
  ;[1 1 1] Apply movement GEs and start timer to check slope.  Set walkable surface to be more forgiving so you can get some halfpipe action.
  (bind _asfort_player_pawn (Utilities|Casting|CastToFortPlayerPawn (Ability|GetAvatarActorfromActorInfo)))
  (Variables|Default|SetPlayerPawn _asfort_player_pawn)
  (Variables|Default|SetDefaultWalkableAngle (Pawn|Components|CharacterMovement|GetWalkableFloorAngle (Variables|/Script/Engine.Character|GetCharacterMovement _playerpawn)))
  (bind _returnvalue_2 (GameplayEffects|ApplyGameplayEffectToSelf :self _returnvalue :GameplayEffectClass "/HighTower/Items/Grape/BrambleShield/GE_HighTower_Grape_Bramble_IcySurface.GE_HighTower_Grape_Bramble_IcySurface_C" :Level 0.0))
  (Variables|Default|SetIcyFeetSurfaceHandle _returnvalue_2)
  (bind _returnvalue_3 (GameplayEffects|ApplyGameplayEffectToSelf :self _returnvalue :GameplayEffectClass "/HighTower/Items/Grape/BrambleShield/GE_HighTower_Grape_BrambleShield_IcySurfaceAcceleration.GE_HighTower_Grape_BrambleShield_IcySurfaceAcceleration_C" :Level 0.0))
  (Variables|Default|SetAccelHandle _returnvalue_3)
  (bind _returnvalue_4 (GameplayEffects|ApplyGameplayEffectToSelf :self _returnvalue :GameplayEffectClass "/HighTower/Items/Grape/BrambleShield/GE_HighTower_Grape_Bramble_Immunity.GE_HighTower_Grape_Bramble_Immunity_C" :Level 0.0))
  (Variables|Default|SetImmunityHandle _returnvalue_4)
  (Pawn|Components|CharacterMovement|SetWalkableFloorAngle (Variables|/Script/Engine.Character|GetCharacterMovement (Variables|Default|GetPlayerPawn)) 84.0)
  (bind _returnvalue_5 (Utilities|Time|SetTimerbyEvent (EventDispatchers|CreateEvent :!event Custom|CheckSlope) 0.1 true))
  (Variables|Default|SetSlopeTimer _returnvalue_5))

(event Ability|EventOnEndAbility
  (params (bWasCancelled :type Boolean))
  (bind _playerpawn (Variables|Default|GetPlayerPawn))
  (bind _returnvalue (Ability|GetAbilitySystemComponent _playerpawn))
  ;[1 1 1] Remove GEs, reset angle var, stop timer
  (GameplayEffects|RemoveActiveGameplayEffect _returnvalue (Variables|Default|GetIcyFeetSurfaceHandle) -1)
  (GameplayEffects|RemoveActiveGameplayEffect _returnvalue (Variables|Default|GetAccelHandle) -1)
  (GameplayEffects|RemoveActiveGameplayEffect _returnvalue (Variables|Default|GetImmunityHandle) -1)
  (Weapon|SetHolsterWeapon _playerpawn 0 false true "None")
  (Pawn|Components|CharacterMovement|SetWalkableFloorAngle (Variables|/Script/Engine.Character|GetCharacterMovement _playerpawn) (Variables|Default|GetDefaultWalkableAngle))
  (Utilities|Time|ClearandInvalidateTimerbyHandle (Variables|Default|GetSlopeTimer)))

(fn SetSpeedVars)
